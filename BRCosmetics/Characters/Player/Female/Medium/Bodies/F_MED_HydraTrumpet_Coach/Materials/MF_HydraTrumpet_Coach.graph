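# Material function for the HydraTrumpet Coach cosmetic body that layers an animated, glowing wrist screen onto incoming material attributes.
# A time-driven flipbook of the wrist screen texture, borrowed from ZombieElasticNeon's clamped flipbook function, combines with sine flicker and a UV-distance vignette through named reroutes.
# These blend background and scan colors, with darkening from the FX2 texture's blue channel.
# The green channel masks the result into Emissive and blacks out Base Color underneath.
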MaterialFunction /BRCosmetics/Characters/Player/Female/Medium/Bodies/F_MED_HydraTrumpet_Coach/Materials/MF_HydraTrumpet_Coach
graph {nodes [
  Constant()
  n1 = GetMaterialAttributes(None=FunctionInput() {inputName=Result, inputType=FunctionInput_MaterialAttributes}) {attributeGetTypes [#1, #2, #3]}
  n2 = TextureSample() {texture=/BRCosmetics/Characters/Player/Female/Medium/Bodies/F_MED_HydraTrumpet_Coach/Textures/T_HydraTrumpet_Coach_Body_FX2}
  n3 = VectorParameter() {group=HydraTrumpet Coach, parameterName=Background Color, defaultValue=<rgba 0.0036 0.289 1 1>}
  n4 = Max(
   A=Multiply(A=@n2.B, B=ScalarParameter() {defaultValue=2, group=HydraTrumpet Coach, parameterName=Screen AO Mask})
   B=ScalarParameter() {defaultValue=0.7, group=HydraTrumpet Coach, parameterName=Min Screen Edge Darkness}
  )
  n5 = NamedRerouteUsage()
  n6 = NamedRerouteUsage()
  n7 = NamedRerouteUsage()
  n8 = ScalarParameter() {defaultValue=0.6, group=HydraTrumpet Coach, parameterName=Flipbook Opacity}
  n9 = LinearInterpolate(
   A=Multiply(A=Multiply(A=@n3.RGB, B=@n4), B=Multiply(A=@n5, B=@n6))
   B=VectorParameter() {group=HydraTrumpet Coach, parameterName=Scan Color, defaultValue=<rgba 0.007 0.94 1.5 1>}.RGB
   Alpha=Multiply(A=NamedRerouteUsage(), B=Multiply(A=@n7, B=@n8))
  )
  FunctionOutput(None=SetMaterialAttributes!(
    MaterialAttributes=@n1.MaterialAttributes
    BaseColor=LinearInterpolate(A=@n1.BaseColor, Alpha=@n2.G) {constB=0}
    EmissiveColor=LinearInterpolate(A=@n1.EmissiveColor, B=@n9, Alpha=@n2.G)
   ) {attributeSetTypes [#1, #3]}) {outputName=SetMaterialAttributes}
  n10 = MaterialFunctionCall!(
   Sine Phase=Multiply(A=Time(), B=ScalarParameter() {defaultValue=4.2353, parameterName=Flicker Speed})
   Value 1=ScalarParameter() {defaultValue=0.85, parameterName=Min Flicker}
   Value 2
  ) {materialFunction=/Engine/Functions/Engine_MaterialFunctions02/Utility/Sine_Remapped}
  NamedRerouteDeclaration(None=Clamp(None=Frac(None=@n10.Remapped Sine)) {minDefault=0.8}) {name=Screen Flicker, nodeColor=<rgba 0 1 0.0625 1>}
  n11 = Length(None=Subtract(A=TextureCoordinate(), B=Constant2Vector() {desc=UV value of screen, g=0.586, r=0.626}) {constB=0.5}) {is=class MaterialExpressionLength}
  n12 = ScalarParameter() {defaultValue=-50, parameterName=Vignette Size}
  NamedRerouteDeclaration(None=Add(A=Multiply(A=@n11, B=@n12), B=ScalarParameter() {defaultValue=1.1594, parameterName=Vignette Strength})) {name=Vignette, nodeColor=<rgba 1 0 0.2578 1>}
  n13 = Time()
  n14 = ScalarParameter() {defaultValue=0.5, group=HydraTrumpet Coach, parameterName=Flipbook Time}
  n15 = TextureObject() {
   samplerType=SAMPLERTYPE_LinearGrayscale
   texture=/BRCosmetics/Characters/Player/Female/Medium/Bodies/F_MED_HydraTrumpet_Coach/Textures/T_HydraTrumpet_Coach_Body_WristScreen
  }
  n16 = VectorParameter() {group=HydraTrumpet Coach, parameterName=Flipbook Position, defaultValue=<rgba 1.1931 2.1673 0 1>}
  n17 = Add(
   A=Multiply(A=TextureCoordinate(), B=ScalarParameter() {defaultValue=50, group=HydraTrumpet Coach, parameterName=Tiling})
   B=MaterialFunctionCall(X=@n16.R, Y=@n16.R) {materialFunction=&MakeFloat2}.Result
  )
  NamedRerouteDeclaration(None=MaterialFunctionCall!(
    "Animation  Phase (0-1)"=Multiply(A=@n13, B=@n14)
    Number of Rows=Constant() {r=8}
    Number of Columns=Constant() {r=3}
    Texture=@n15
    UVs=Frac(None=@n17)
    MipBias/Level=ScalarParameter() {defaultValue=-2, group=HydraTrumpet Coach, parameterName="MipBias/Level (S)"}
    "Use Mip Bias (T) Level (F)"
    "Clamp Anim (see tooltip)"
   ) {materialFunction=/BRCosmetics/Characters/Player/Female/Medium/Bodies/F_MED_ZombieElasticNeon/FX/FlipBook_Clamped}.Result) {name=Flipbook, nodeColor=<rgba 1 0.1406 0 1>}
  NamedRerouteDeclaration(None=@n2.R) {name=Flipbook Mask, nodeColor=<rgba 0 0.1797 1 1>}
  VectorParameter() {group=HydraTrumpet Coach, parameterName=Flipbook Color, defaultValue=<rgba 0.9911 0.5776 0.0009 1>}
  VectorParameter() {group=HydraTrumpet Coach, parameterName=Screen Color, defaultValue=<rgba 0.9216 0.0242 0.0006 1>}
 ]}
