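# Control Rig for the HydraTrumpet Coach female body that keeps a bandolier and hip bag fitted to the moving character.
# Weighted parent constraints attach strap control nulls to body bones, and Spline IK drives strap bone chains from controls.
# Spherical pose readers on arms, spine, pelvis and right thigh weight corrective transform offsets.
# Smoothed state variables like Crouch, Emote, Jump, KneeSlide and pickaxe add or scale further offsets.
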ControlRigBlueprint /BRCosmetics/Characters/Player/Female/Medium/Bodies/F_MED_HydraTrumpet_Coach/Meshes/F_MED_HydraTrumpet_Coach_CtrlRig
graphs [
 New Function_ContainedGraph {nodes [
   RigVMFunctionEntryNode()
   > n1 = &RigUnit_SphericalPoseReader(
    ActiveRegionScaleFactors=(PositiveWidth=2,NegativeWidth=1,PositiveHeight=1,NegativeHeight=1)
    ActiveRegionSize=0.12
    Debug=(bDrawDebug=false,bDraw2D=false,DebugScale=25,DebugSegments=20,DebugThickness=0.25)
    DriverAxis=(X=1,Y=0,Z=0)
    DriverItem=(Type=Bone,Name="thumb_03_l")
    FalloffRegionScaleFactors=(PositiveWidth=-2,NegativeWidth=1,PositiveHeight=1,NegativeHeight=1)
    FalloffSize=0.2
    OptionalParentItem=(Type=Bone,Name="lowerarm_l")
    RotationOffset=(X=-30,Y=30,Z=20)
   )
   > &RigUnit_ModifyTransforms(
    ItemToModify="((Item=(Type=Bone,Name=\"lowerarm_twist_01_l\"),Transform=(Rotation=(X=0,Y=-0.0262,Z=0,W=0.9997),Translation=(X=-0.7,Y=-0.17,Z=1),Scale3D=(X=1,Y=1,Z=1))),(Item=(Type=Bone,Name=\"lowerarm_twist_02_l\"),Transform=(Rotation=(X=0.342,Y=0,Z=0,W=0.9397),Translation= ...(truncated)"
    WeightMinimum=0
    Weight=@n1.OutputParam
   )
   > RigVMFunctionReturnNode()
  ]}
 New Function_ContainedGraph {nodes [
   RigVMFunctionEntryNode()
   > &RigUnit_ParentConstraint(
    AdvancedSettings=(RotationOrderForFilter=XZY)
    Child=(Type=Bone,Name="base_bandolier")
    Parents="((Item=(Type=Bone,Name=\"neck_01\"),Weight=0.001),(Item=(Type=Bone,Name=\"spine_04\"),Weight=0.0077),(Item=(Type=Bone,Name=\"deform_upperarm_bk_l\"),Weight=0.0315),(Item=(Type=Bone,Name=\"deform_upperarm_fr_l\"),Weight=0.0319),(Item=(Type=Bone,Name=\"deform_upperar ...(truncated)"
    Weight=1
   )
   > &RigUnit_ParentConstraint(
    AdvancedSettings=(RotationOrderForFilter=XZY)
    Child=(Type=Null,Name="def_bandolier_front_01_ctrl_null")
    Parents="((Item=(Type=Bone,Name=\"deform_pec_l\"),Weight=0.0088),(Item=(Type=Bone,Name=\"deform_upperarm_bk_l\"),Weight=0.0155),(Item=(Type=Bone,Name=\"spine_04\"),Weight=0.0275),(Item=(Type=Bone,Name=\"deform_upperarm_fr_l\"),Weight=0.0805),(Item=(Type=Bone,Name=\"deform_u ...(truncated)"
    Weight=1
   )
   > &RigUnit_ParentConstraint(
    AdvancedSettings=(RotationOrderForFilter=XZY)
    Child=(Type=Null,Name="def_bandolier_front_02_ctrl_null")
    Parents="((Item=(Type=Bone,Name=\"deform_upperarm_bk_l\"),Weight=0.0026),(Item=(Type=Bone,Name=\"deform_pec_l\"),Weight=0.0479),(Item=(Type=Bone,Name=\"deform_upperarm_up_l\"),Weight=0.0575),(Item=(Type=Bone,Name=\"spine_04\"),Weight=0.0692),(Item=(Type=Bone,Name=\"spine_05 ...(truncated)"
    Weight=1
   )
   > &RigUnit_ParentConstraint(
    AdvancedSettings=(RotationOrderForFilter=XZY)
    Child=(Type=Null,Name="def_bandolier_front_03_ctrl_null")
    Parents="((Item=(Type=Bone,Name=\"deform_pec_aim_l\"),Weight=0.0009),(Item=(Type=Bone,Name=\"deform_upperarm_up_l\"),Weight=0.0027),(Item=(Type=Bone,Name=\"upperarm_twist_01_l\"),Weight=0.0764),(Item=(Type=Bone,Name=\"deform_upperarm_fr_l\"),Weight=0.0774),(Item=(Type=Bone ...(truncated)"
    Weight=1
   )
   > &RigUnit_ParentConstraint(
    AdvancedSettings=(RotationOrderForFilter=XZY)
    Child=(Type=Null,Name="def_bandolier_front_04_ctrl_null")
    Parents="((Item=(Type=Bone,Name=\"deform_upperarm_fr_l\"),Weight=0.0023),(Item=(Type=Bone,Name=\"upperarm_twist_01_l\"),Weight=0.0037),(Item=(Type=Bone,Name=\"deform_pec_aim_r\"),Weight=0.0038),(Item=(Type=Bone,Name=\"clavicle_l\"),Weight=0.0082),(Item=(Type=Bone,Name=\"spi ...(truncated)"
    Weight=1
   )
   > &RigUnit_ParentConstraint(
    AdvancedSettings=(RotationOrderForFilter=XZY)
    Child=(Type=Null,Name="def_bandolier_front_05_ctrl_null")
    Parents="((Item=(Type=Bone,Name=\"spine_05\"),Weight=0.0006),(Item=(Type=Bone,Name=\"deform_pec_aim_r\"),Weight=0.0014),(Item=(Type=Bone,Name=\"deform_pec_l\"),Weight=0.0025),(Item=(Type=Bone,Name=\"spine_04\"),Weight=0.004),(Item=(Type=Bone,Name=\"deform_pec_aim_l\"),Weight ...(truncated)"
    Weight=1
   )
   > &RigUnit_ParentConstraint(
    AdvancedSettings=(RotationOrderForFilter=XZY)
    Child=(Type=Null,Name="def_bandolier_front_06_ctrl_null")
    Parents="((Item=(Type=Bone,Name=\"pelvis\"),Weight=0.0003),(Item=(Type=Bone,Name=\"deform_pec_aim_r\"),Weight=0.0009),(Item=(Type=Bone,Name=\"spine_01\"),Weight=0.0161),(Item=(Type=Bone,Name=\"spine_02\"),Weight=0.0573),(Item=(Type=Bone,Name=\"spine_03\"),Weight=0.0701),(Ite ...(truncated)"
    Weight=1
   )
   > &RigUnit_ParentConstraint(
    AdvancedSettings=(RotationOrderForFilter=XZY)
    Child=(Type=Null,Name="def_bandolier_front_07_ctrl_null")
    Parents="((Item=(Type=Bone,Name=\"deform_pec_aim_r\"),Weight=0.0036),(Item=(Type=Bone,Name=\"deform_groin_l\"),Weight=0.0103),(Item=(Type=Bone,Name=\"deform_groin_r\"),Weight=0.0142),(Item=(Type=Bone,Name=\"pelvis\"),Weight=0.1054),(Item=(Type=Bone,Name=\"spine_03\"),Weight= ...(truncated)"
    Weight=1
   )
   > &RigUnit_ParentConstraint(
    AdvancedSettings=(RotationOrderForFilter=XZY)
    Child=(Type=Null,Name="def_bandolier_rear_01_ctrl_null")
    Parents="((Item=(Type=Bone,Name=\"neck_01\"),Weight=0.0013),(Item=(Type=Bone,Name=\"spine_04\"),Weight=0.0039),(Item=(Type=Bone,Name=\"deform_upperarm_fr_l\"),Weight=0.0081),(Item=(Type=Bone,Name=\"deform_upperarm_up_l\"),Weight=0.0675),(Item=(Type=Bone,Name=\"deform_uppera ...(truncated)"
    Weight=1
   )
   > &RigUnit_ParentConstraint(
    AdvancedSettings=(RotationOrderForFilter=XZY)
    Child=(Type=Null,Name="def_bandolier_rear_02_ctrl_null")
    Parents="((Item=(Type=Bone,Name=\"spine_03\"),Weight=0.0008),(Item=(Type=Bone,Name=\"clavicle_r\"),Weight=0.0062),(Item=(Type=Bone,Name=\"deform_upperarm_up_l\"),Weight=0.0335),(Item=(Type=Bone,Name=\"spine_04\"),Weight=0.0463),(Item=(Type=Bone,Name=\"deform_upperarm_bk_l\") ...(truncated)"
    Weight=1
   )
   > &RigUnit_ParentConstraint(
    AdvancedSettings=(RotationOrderForFilter=XZY)
    Child=(Type=Null,Name="def_bandolier_rear_03_ctrl_null")
    Parents="((Item=(Type=Bone,Name=\"deform_upperarm_up_l\"),Weight=0.0036),(Item=(Type=Bone,Name=\"spine_03\"),Weight=0.007),(Item=(Type=Bone,Name=\"clavicle_r\"),Weight=0.0351),(Item=(Type=Bone,Name=\"upperarm_twist_01_l\"),Weight=0.0863),(Item=(Type=Bone,Name=\"deform_upper ...(truncated)"
    Weight=1
   )
   > &RigUnit_ParentConstraint(
    AdvancedSettings=(RotationOrderForFilter=XZY)
    Child=(Type=Null,Name="def_bandolier_rear_04_ctrl_null")
    Parents="((Item=(Type=Bone,Name=\"spine_02\"),Weight=0.0247),(Item=(Type=Bone,Name=\"clavicle_r\"),Weight=0.025),(Item=(Type=Bone,Name=\"spine_03\"),Weight=0.0301),(Item=(Type=Bone,Name=\"upperarm_twist_01_l\"),Weight=0.0744),(Item=(Type=Bone,Name=\"deform_upperarm_bk_l\"),W ...(truncated)"
    Weight=1
   )
   > &RigUnit_ParentConstraint(
    AdvancedSettings=(RotationOrderForFilter=XZY)
    Child=(Type=Null,Name="def_bandolier_rear_05_ctrl_null")
    Parents="((Item=(Type=Bone,Name=\"spine_01\"),Weight=0.0052),(Item=(Type=Bone,Name=\"deform_upperarm_bk_l\"),Weight=0.0098),(Item=(Type=Bone,Name=\"upperarm_twist_01_l\"),Weight=0.0102),(Item=(Type=Bone,Name=\"spine_02\"),Weight=0.0739),(Item=(Type=Bone,Name=\"spine_03\"),We ...(truncated)"
    Weight=1
   )
   > &RigUnit_ParentConstraint(
    AdvancedSettings=(RotationOrderForFilter=XZY)
    Child=(Type=Null,Name="def_bandolier_rear_06_ctrl_null")
    Parents="((Item=(Type=Bone,Name=\"clavicle_r\"),Weight=0.0044),(Item=(Type=Bone,Name=\"pelvis\"),Weight=0.0093),(Item=(Type=Bone,Name=\"spine_05\"),Weight=0.0169),(Item=(Type=Bone,Name=\"spine_01\"),Weight=0.1165),(Item=(Type=Bone,Name=\"spine_04\"),Weight=0.2461),(Item=(Typ ...(truncated)"
    Weight=1
   )
   > RigVMFunctionReturnNode()
  ]}
 New Function_ContainedGraph {nodes [
   RigVMFunctionEntryNode()
   > RigVMFunction_Sequence_1 = RigVMAggregateNode()
   n1 = &RigVMFunction_AlphaInterp(
    InterpSpeedDecreasing=10
    InterpSpeedIncreasing=5
    Scale=1
    bClampResult=true
    bMapRange=true
    Value=RigVMVariableNode() {variable=Crouch}.Value
   )
   &RigUnit_ModifyTransforms(
    ItemToModify="((Item=(Type=Bone,Name=\"dyn_bag_sim_chain_03\"),Transform=(Rotation=(X=-0.0872,Y=0,Z=0,W=0.9962),Translation=(X=5,Y=0,Z=0),Scale3D=(X=1,Y=1,Z=1))),(Item=(Type=Bone,Name=\"base_bag_rear_01\"),Transform=(Rotation=(X=0,Y=0.1305,Z=0,W=0.9914),Translation=(X=0,Y=0 ...(truncated)"
    WeightMinimum=0
    Weight=@n1.Result
    ExecutePin=@RigVMFunction_Sequence_1.A
   )
   > n2 = &RigUnit_SphericalPoseReader(
    ActiveRegionScaleFactors=(PositiveWidth=1,NegativeWidth=3.5,PositiveHeight=1.5,NegativeHeight=1.5)
    ActiveRegionSize=0.1
    Debug=(bDrawDebug=false,DebugScale=25,DebugSegments=20,DebugThickness=0.25)
    DriverAxis=(X=1,Y=0,Z=0)
    DriverItem=(Type=Bone,Name="thigh_r")
    FalloffRegionScaleFactors=(PositiveWidth=1,NegativeWidth=1,PositiveHeight=1,NegativeHeight=1)
    FalloffSize=0.35
    OptionalParentItem=(Type=Bone,Name="pelvis")
    RotationOffset=(X=-5,Y=-27,Z=0)
   )
   > &RigUnit_ModifyTransforms(
    ItemToModify="((Item=(Type=Bone,Name=\"dyn_bag_sim_chain_03\"),Transform=(Rotation=(X=0,Y=0,Z=0,W=1),Translation=(X=0,Y=0,Z=0),Scale3D=(X=1,Y=1,Z=1))),(Item=(Type=Bone,Name=\"dyn_bag_sim_chain_01\"),Transform=(Rotation=(X=0,Y=-0.1736,Z=0,W=0.9848),Translation=(X=2,Y=0,Z=0), ...(truncated)"
    WeightMinimum=0
    Weight=@n2.OutputParam
   )
   > n3 = &RigUnit_SphericalPoseReader(
    ActiveRegionScaleFactors=(PositiveWidth=1,NegativeWidth=1,PositiveHeight=1,NegativeHeight=1)
    ActiveRegionSize=0.1
    Debug=(bDrawDebug=false,bDraw2D=false,DebugScale=25,DebugSegments=20,DebugThickness=0.25)
    DriverAxis=(X=1,Y=0,Z=0)
    DriverItem=(Type=Bone,Name="spine_05")
    FalloffRegionScaleFactors=(PositiveWidth=1,NegativeWidth=1,PositiveHeight=1,NegativeHeight=1)
    FalloffSize=0.24
    OptionalParentItem=(Type=Bone,Name="pelvis")
    RotationOffset=(X=90,Y=0,Z=-110)
   )
   > &RigUnit_ModifyTransforms(
    ItemToModify="((Item=(Type=Bone,Name=\"dyn_bag_sim_chain_03\"),Transform=(Rotation=(X=0,Y=0,Z=0.0436,W=0.999),Translation=(X=0,Y=3,Z=0),Scale3D=(X=1,Y=1,Z=1))),(Item=(Type=Null,Name=\"def_bandolier_rear_04_ctrl_null\"),Transform=(Rotation=(X=0,Y=0,Z=-0.0698,W=0.9976),Transl ...(truncated)"
    WeightMinimum=0
    Weight=@n3.OutputParam
   )
   > n4 = &RigUnit_SphericalPoseReader(
    ActiveRegionScaleFactors=(PositiveWidth=1,NegativeWidth=1,PositiveHeight=1,NegativeHeight=1)
    ActiveRegionSize=0.1
    Debug=(bDrawDebug=false,bDraw2D=false,bDrawLocalAxes=False,DebugScale=25,DebugSegments=20,DebugThickness=0.25)
    DriverAxis=(X=-1,Y=0,Z=0)
    DriverItem=(Type=Bone,Name="thigh_r")
    FalloffRegionScaleFactors=(PositiveWidth=0.3,NegativeWidth=0.3,PositiveHeight=0.3,NegativeHeight=0.3)
    FalloffSize=0.45
    FlipHeightScaling=False
    FlipWidthScaling=False
    OptionalParentItem=(Type=Bone,Name="None")
    RotationOffset=(X=20,Y=150,Z=-25)
   )
   > &RigUnit_ModifyTransforms(
    ItemToModify=((Item=(Type=Bone,Name="dyn_bag_sim_chain_01"),Transform=(Rotation=(X=0,Y=-0.2164,Z=0,W=0.9763),Translation=(X=0,Y=0,Z=0),Scale3D=(X=1,Y=1,Z=1))))
    WeightMinimum=0
    Weight=@n4.OutputParam
   )
   > &RigUnit_ParentConstraint(
    AdvancedSettings=(InterpolationType=Average,RotationOrderForFilter=XZY)
    Child=(Type=Null,Name="dyn_bag_sim_chain_03_ctrl_null")
    Parents=((Item=(Type=Bone,Name="pelvis"),Weight=1),(Item=(Type=Bone,Name="thigh_r"),Weight=1))
    Weight=1
   )
   > SplineIK = RigVMFunctionReferenceNode(
    Primary Axis=(X=1,Y=0,Z=0)
    Samples Per Segment=6
    Secondary Spline Direction=(X=0,Y=1,Z=0)
    Spline Mode=BSpline
    Stretch=false
    Up Axis=(X=0,Y=1,Z=0)
    Controls=RigVMVariableNode() {variable=bandolier_front_ctrl}.Value
    Bones=RigVMVariableNode() {variable=bandolier_front_joints}.Value
   )
   > SplineIK_1 = RigVMFunctionReferenceNode(
    Primary Axis=(X=1,Y=0,Z=0)
    Samples Per Segment=6
    Secondary Spline Direction=(X=0,Y=-1,Z=0)
    Spline Mode=BSpline
    Stretch=false
    Up Axis=(X=0,Y=-1,Z=0)
    Controls=RigVMVariableNode() {variable=bandolier_rear_ctrl}.Value
    Bones=RigVMVariableNode() {variable=bandolier_rear_joints}.Value
   )
   n5 = &RigUnit_SphericalPoseReader(
    ActiveRegionScaleFactors=(PositiveWidth=0.5,NegativeWidth=2,PositiveHeight=4.5,NegativeHeight=3)
    ActiveRegionSize=0.1
    Debug=(bDrawDebug=false,DebugScale=25,DebugSegments=20,DebugThickness=0.25)
    DriverAxis=(X=1,Y=0,Z=0)
    DriverItem=(Type=Bone,Name="thigh_r")
    FalloffRegionScaleFactors=(PositiveWidth=0.6,NegativeWidth=1,PositiveHeight=1,NegativeHeight=1)
    FalloffSize=0.45
    OptionalParentItem=(Type=Bone)
    RotationOffset=(X=-60,Y=0,Z=10)
    ExecutePin=@RigVMFunction_Sequence_1.B
   )
   > &RigUnit_ModifyTransforms(
    ItemToModify=((Item=(Type=Bone,Name="dyn_bag_sim_chain_01"),Transform=(Rotation=(X=0,Y=-0.0436,Z=0,W=0.999),Translation=(X=-5,Y=0,Z=0),Scale3D=(X=1,Y=1,Z=1))))
    WeightMinimum=0
    Weight=@n5.OutputParam
   )
   n6 = &RigVMFunction_AlphaInterp(
    InterpSpeedDecreasing=5
    InterpSpeedIncreasing=5
    Scale=1
    bClampResult=true
    bMapRange=true
    Value=RigVMVariableNode() {variable=CrouchMOV}.Value
   )
   > &RigUnit_ModifyTransforms(
    ItemToModify="((Item=(Type=Bone,Name=\"dyn_bandolier_front_05\"),Transform=(Rotation=(X=0,Y=0.2164,Z=0,W=0.9763),Translation=(X=0,Y=0,Z=0),Scale3D=(X=1,Y=1,Z=1))),(Item=(Type=Bone,Name=\"dyn_bag_sim_chain_02\"),Transform=(Rotation=(X=0,Y=0,Z=0,W=1),Translation=(X=0,Y=0,Z=0) ...(truncated)"
    WeightMinimum=0
    Weight=&RigVMFunction_MathDoubleMul(B=@n6.Result, A=@n5.OutputParam).Result
   )
   n7 = &RigVMFunction_AlphaInterp(
    InterpSpeedDecreasing=5
    InterpSpeedIncreasing=5
    Scale=1
    bClampResult=true
    bMapRange=true
    Value=RigVMVariableNode() {variable=Emote}.Value
   )
   > &RigUnit_ModifyTransforms(
    ItemToModify="((Item=(Type=Bone,Name=\"dyn_bag_sim_chain_01\"),Transform=(Rotation=(X=-0.3173,Y=0,Z=0,W=0.9483),Translation=(X=0,Y=4,Z=-4),Scale3D=(X=1,Y=1,Z=1))),(Item=(Type=Bone,Name=\"dyn_bandolier_rear_06\"),Transform=(Rotation=(X=0,Y=0,Z=0,W=1),Translation=(X=0,Y=0,Z=0 ...(truncated)"
    WeightMinimum=0
    Weight=@n7.Result
   )
   n8 = &RigVMFunction_AlphaInterp(
    InterpSpeedDecreasing=10
    InterpSpeedIncreasing=10
    Scale=1
    bClampResult=False
    bInterpResult=False
    Value=RigVMVariableNode() {variable=grind}.Value
   )
   > &RigUnit_ModifyTransforms(
    ItemToModify="((Item=(Type=Bone,Name=\"dyn_bandolier_rear_06\"),Transform=(Rotation=(X=0,Y=0,Z=0.2588,W=0.9659),Translation=(X=0,Y=0,Z=0),Scale3D=(X=1,Y=1,Z=1))),(Item=(Type=Bone,Name=\"dyn_bag_sim_chain_01\"),Transform=(Rotation=(X=0,Y=0,Z=0,W=1),Translation=(X=-5,Y=0,Z=0) ...(truncated)"
    WeightMinimum=0
    Weight=@n8.Result
   )
   n9 = &RigVMFunction_AlphaInterp(
    InterpSpeedDecreasing=10
    InterpSpeedIncreasing=10
    Scale=1
    bClampResult=False
    bInterpResult=False
    Value=RigVMVariableNode() {variable=KneeSlide}.Value
   )
   > &RigUnit_ModifyTransforms(
    ItemToModify="((Item=(Type=Bone,Name=\"dyn_bag_sim_chain_01\"),Transform=(Rotation=(X=-0.171,Y=0.0302,Z=-0.171,W=0.9698),Translation=(X=-5,Y=0,Z=0),Scale3D=(X=1,Y=1,Z=1))),(Item=(Type=Bone,Name=\"dyn_bandolier_rear_04\"),Transform=(Rotation=(X=0,Y=0,Z=0.0872,W=0.9962),Trans ...(truncated)"
    WeightMinimum=0
    Weight=@n9.Result
   )
   n10 = &RigVMFunction_AlphaInterp(InterpSpeedDecreasing=2, InterpSpeedIncreasing=2, Scale=1, bClampResult=False, Value=RigVMVariableNode() {variable=Jump}.Value)
   > &RigUnit_ModifyTransforms(
    ItemToModify="((Item=(Type=Bone,Name=\"dyn_bag_sim_chain_01\"),Transform=(Rotation=(X=0.067,Y=-0.25,Z=0.25,W=0.933),Translation=(X=0,Y=0,Z=0),Scale3D=(X=1,Y=1,Z=1))),(Item=(Type=Bone,Name=\"dyn_bag_sim_chain_03\"),Transform=(Rotation=(X=0,Y=0,Z=0,W=1),Translation=(X=0,Y=0,Z ...(truncated)"
    WeightMinimum=0
    Weight=@n10.Result
   )
   > &RigUnit_ModifyTransforms(
    ItemToModify="((Item=(Type=Bone,Name=\"dyn_bag_sim_chain_01\"),Transform=(Rotation=(X=0.067,Y=-0.25,Z=0.25,W=0.933),Translation=(X=0,Y=0,Z=0),Scale3D=(X=1,Y=1,Z=1))),(Item=(Type=Bone,Name=\"dyn_bag_sim_chain_03\"),Transform=(Rotation=(X=0,Y=0,Z=0,W=1),Translation=(X=0,Y=0,Z ...(truncated)"
    WeightMinimum=0
    Weight=RigVMVariableNode() {variable=pinLowerBack}.Value
   )
   RigVMFunction_MathBoolToFloat = &RigVMFunction_MathBoolToFloat(Value=RigVMVariableNode() {variable=FE}.Value)
   n11 = &RigVMFunction_AlphaInterp(InterpSpeedDecreasing=5, InterpSpeedIncreasing=5, Scale=1, bClampResult=true, bMapRange=true, Value=@RigVMFunction_MathBoolToFloat.Result)
   > &RigUnit_ModifyTransforms(
    ItemToModify="((Item=(Type=Bone,Name=\"dyn_bandolier_front_05\"),Transform=(Rotation=(X=0,Y=0,Z=-0.0436,W=0.999),Translation=(X=0,Y=-2,Z=0),Scale3D=(X=1,Y=1,Z=1))),(Item=(Type=Bone,Name=\"dyn_bandolier_rear_05\"),Transform=(Rotation=(X=0,Y=0,Z=0.1736,W=0.9848),Translation=( ...(truncated)"
    WeightMinimum=0
    Weight=@n11.Result
   )
   RigVMFunctionReturnNode(ExecuteContext=@RigVMFunction_Sequence_1.C)
  ]}
 New Function_ContainedGraph {nodes [
   n1 = RigVMFunctionEntryNode()
   RigVMFunction_ControlFlowBranch = &RigVMFunction_ControlFlowBranch(Condition=@n1.Debug, ExecuteContext=@n1.ExecuteContext)
   RigVMFunctionReturnNode(ExecuteContext=@RigVMFunction_ControlFlowBranch.Completed)
   &RigUnit_SetTransform(
    Item=(Type=Bone,Name="thigh_r")
    Weight=1
    bInitial=False
    Value=&RigUnit_GetTransform(Item=(Type=Control,Name="thigh_r_ctrl"), bInitial=False).Transform
    ExecutePin=@RigVMFunction_ControlFlowBranch.True
   )
  ]}
 New Function_ContainedGraph {nodes [
   RigVMFunctionEntryNode()
   > RigVMFunction_Sequence = RigVMAggregateNode()
   n1 = &RigUnit_SphericalPoseReader(
    ActiveRegionScaleFactors=(PositiveWidth=2.5,NegativeWidth=3,PositiveHeight=3,NegativeHeight=3)
    ActiveRegionSize=0.1
    Debug=(bDrawDebug=false,DebugScale=25,DebugSegments=20,DebugThickness=0.25)
    DriverAxis=(X=1,Y=0,Z=0)
    DriverItem=(Type=Bone,Name="upperarm_l")
    FalloffRegionScaleFactors=(PositiveWidth=1,NegativeWidth=1,PositiveHeight=1,NegativeHeight=1)
    FalloffSize=0.5
    FlipHeightScaling=false
    FlipWidthScaling=false
    OptionalParentItem=(Type=Bone)
    RotationOffset=(X=0,Y=0,Z=0)
    ExecutePin=@RigVMFunction_Sequence.A
   )
   > n2 = &RigUnit_SphericalPoseReader(
    ActiveRegionScaleFactors=(PositiveWidth=1,NegativeWidth=1,PositiveHeight=2,NegativeHeight=2)
    ActiveRegionSize=0.1
    Debug=(bDrawDebug=false,DebugScale=25,DebugSegments=20,DebugThickness=0.25)
    DriverAxis=(X=1,Y=0,Z=0)
    DriverItem=(Type=Bone,Name="spine_03")
    FalloffRegionScaleFactors=(PositiveWidth=0.1,NegativeWidth=0.1,PositiveHeight=0.5,NegativeHeight=0.5)
    FalloffSize=0.5
    FlipHeightScaling=false
    FlipWidthScaling=false
    OptionalParentItem=(Type=Bone)
    RotationOffset=(X=-90,Y=0,Z=65)
   )
   > n3 = &RigUnit_SphericalPoseReader(
    ActiveRegionScaleFactors=(PositiveWidth=1,NegativeWidth=1,PositiveHeight=1,NegativeHeight=1)
    ActiveRegionSize=0.1
    Debug=(bDrawDebug=false,DebugScale=25,DebugSegments=20,DebugThickness=0.25)
    DriverAxis=(X=1,Y=0,Z=0)
    DriverItem=(Type=Bone,Name="upperarm_l")
    FalloffRegionScaleFactors=(PositiveWidth=1,NegativeWidth=1,PositiveHeight=1,NegativeHeight=1)
    FalloffSize=0.35
    OptionalParentItem=(Type=Bone,Name="spine_05")
    RotationOffset=(X=7.0096,Y=-141.722,Z=16.681)
   )
   > &RigUnit_ModifyTransforms(
    ItemToModify="((Item=(Type=Bone,Name=\"dyn_bag_sim_chain_01\"),Transform=(Rotation=(X=0,Y=0.165,Z=0,W=0.9863),Translation=(X=0,Y=0,Z=0),Scale3D=(X=1,Y=1,Z=1))),(Item=(Type=Bone,Name=\"dyn_bandolier_rear_06\"),Transform=(Rotation=(X=0.3062,Y=-0.1768,Z=0.3062,W=0.8839),Transl ...(truncated)"
    WeightMinimum=0
    Weight=@n3.OutputParam
   )
   n4 = &RigVMFunction_AlphaInterp(
    InterpSpeedDecreasing=5
    InterpSpeedIncreasing=5
    Scale=1
    bClampResult=true
    bMapRange=true
    Value=RigVMVariableNode() {variable=Crouch}.Value
   )
   > &RigUnit_ModifyTransforms(
    ItemToModify="((Item=(Type=Bone,Name=\"dyn_bag_sim_chain_01\"),Transform=(Rotation=(X=0,Y=-0.1392,Z=0,W=0.9903),Translation=(X=0,Y=0,Z=0),Scale3D=(X=1,Y=1,Z=1))),(Item=(Type=Bone,Name=\"dyn_bandolier_front_05\"),Transform=(Rotation=(X=0,Y=0,Z=0,W=1),Translation=(X=0,Y=0,Z=0 ...(truncated)"
    WeightMinimum=0
    Weight=&RigVMFunction_MathDoubleMul(B=@n4.Result, A=@n3.OutputParam).Result
   )
   > &RigUnit_ModifyTransforms(
    ItemToModify=((Item=(Type=Bone,Name="dyn_bandolier_rear_05"),Transform=(Rotation=(X=0,Y=0,Z=-0.2588,W=0.9659),Translation=(X=0,Y=-3,Z=0),Scale3D=(X=1,Y=1,Z=1))))
    WeightMinimum=0
    Weight=@n2.OutputParam
   )
   > &RigUnit_ModifyTransforms(
    ItemToModify="((Item=(Type=Bone,Name=\"dyn_bag_sim_chain_03\"),Transform=(Rotation=(X=0,Y=0,Z=0,W=1),Translation=(X=-4,Y=0,Z=5),Scale3D=(X=1,Y=1,Z=1))),(Item=(Type=Bone,Name=\"dyn_bandolier_front_05\"),Transform=(Rotation=(X=0,Y=0,Z=0.1736,W=0.9848),Translation=(X=0,Y=0,Z=0 ...(truncated)"
    WeightMinimum=0
    Weight=@n1.OutputParam
   )
   n5 = &RigVMFunction_AlphaInterp(
    InterpSpeedDecreasing=3
    InterpSpeedIncreasing=10
    Scale=1
    bClampResult=true
    bMapRange=true
    Value=RigVMVariableNode() {variable=pickaxe}.Value
   )
   > &RigUnit_ModifyTransforms(
    ItemToModify="((Item=(Type=Bone,Name=\"dyn_bag_sim_chain_01\"),Transform=(Rotation=(X=0,Y=-0.2334,Z=0,W=0.9724),Translation=(X=0,Y=0,Z=0),Scale3D=(X=1,Y=1,Z=1))),(Item=(Type=Bone,Name=\"base_bandolier\"),Transform=(Rotation=(X=0,Y=0,Z=0,W=1),Translation=(X=0,Y=0,Z=0),Scale3 ...(truncated)"
    WeightMinimum=0
    Weight=@n5.Result
   )
   RigVMFunction_ControlFlowBranch = &RigVMFunction_ControlFlowBranch(ExecuteContext=@RigVMFunction_Sequence.B, Condition=RigVMVariableNode() {variable=FE}.Value)
   n6 = &RigVMFunction_AlphaInterp(
    InterpSpeedDecreasing=10
    InterpSpeedIncreasing=5
    Scale=1
    bClampResult=true
    bMapRange=true
    Value=RigVMVariableNode() {variable=Spine_Rot_L}.Value
   )
   &RigUnit_ModifyTransforms(
    ItemToModify="((Item=(Type=Bone,Name=\"dyn_bag_sim_chain_03\"),Transform=(Rotation=(X=0,Y=0,Z=0,W=1),Translation=(X=0,Y=0,Z=5),Scale3D=(X=1,Y=1,Z=1))),(Item=(Type=Bone,Name=\"dyn_bandolier_front_05\"),Transform=(Rotation=(X=0,Y=0.0872,Z=0,W=0.9962),Translation=(X=0,Y=2,Z=0) ...(truncated)"
    WeightMinimum=0
    Weight=@n6.Result
    ExecutePin=@RigVMFunction_ControlFlowBranch.False
   )
   n7 = &RigVMFunction_AlphaInterp(
    InterpSpeedDecreasing=10
    InterpSpeedIncreasing=10
    Scale=1
    bClampResult=true
    bMapRange=true
    Value=RigVMVariableNode() {variable=Spine_Rot_r}.Value
   )
   > &RigUnit_ModifyTransforms(
    ItemToModify="((Item=(Type=Bone,Name=\"dyn_bag_sim_chain_03\"),Transform=(Rotation=(X=0,Y=0,Z=0,W=1),Translation=(X=0,Y=0,Z=-3),Scale3D=(X=1,Y=1,Z=1))),(Item=(Type=Bone,Name=\"dyn_bandolier_front_05\"),Transform=(Rotation=(X=0,Y=0,Z=0,W=1),Translation=(X=0,Y=2,Z=0),Scale3D= ...(truncated)"
    WeightMinimum=0
    Weight=@n7.Result
   )
   n8 = &RigVMFunction_AlphaInterp(
    InterpSpeedDecreasing=1
    InterpSpeedIncreasing=1
    Scale=1
    bClampResult=true
    bMapRange=true
    Value=RigVMVariableNode() {variable=Spine_Rot_L}.Value
   )
   &RigUnit_ModifyTransforms(
    ItemToModify="((Item=(Type=Bone,Name=\"dyn_bandolier_front_05\"),Transform=(Rotation=(X=0,Y=0.0872,Z=0,W=0.9962),Translation=(X=0,Y=2,Z=0),Scale3D=(X=1,Y=1,Z=1))),(Item=(Type=Bone,Name=\"dyn_bandolier_rear_06\"),Transform=(Rotation=(X=0.0012,Y=0.0349,Z=-0.0349,W=0.9988),Tra ...(truncated)"
    WeightMinimum=0
    Weight=@n8.Result
    ExecutePin=@RigVMFunction_ControlFlowBranch.True
   )
   n9 = &RigVMFunction_AlphaInterp(
    InterpSpeedDecreasing=1
    InterpSpeedIncreasing=1
    Scale=1
    bClampResult=true
    bMapRange=true
    Value=RigVMVariableNode() {variable=Spine_Rot_r}.Value
   )
   > &RigUnit_ModifyTransforms(
    ItemToModify="((Item=(Type=Bone,Name=\"dyn_bandolier_front_05\"),Transform=(Rotation=(X=0,Y=0,Z=0,W=1),Translation=(X=0,Y=2,Z=0),Scale3D=(X=1,Y=1,Z=1))),(Item=(Type=Bone,Name=\"dyn_bandolier_rear_06\"),Transform=(Rotation=(X=0,Y=0.0872,Z=0,W=0.9962),Translation=(X=0,Y=0,Z=0 ...(truncated)"
    WeightMinimum=0
    Weight=@n9.Result
   )
   RigVMFunctionReturnNode(ExecuteContext=@RigVMFunction_Sequence.C)
  ]}
 New Function_ContainedGraph {nodes [
   n1 = RigVMFunctionEntryNode()
   RigVMFunction_Sequence = &RigVMFunction_Sequence(ExecuteContext=@n1.ExecuteContext)
   n2 = &RigUnit_SphericalPoseReader(
    ActiveRegionScaleFactors=(PositiveWidth=1,NegativeWidth=1,PositiveHeight=1,NegativeHeight=1)
    ActiveRegionSize=0.1
    Debug=(bDrawDebug=false,DebugScale=25,DebugSegments=20,DebugThickness=0.25)
    DriverAxis=(X=1,Y=0,Z=0)
    DriverItem=(Type=Bone,Name="upperarm_l")
    FalloffRegionScaleFactors=(PositiveWidth=1,NegativeWidth=1,PositiveHeight=1,NegativeHeight=1)
    FalloffSize=0.35
    OptionalParentItem=(Type=Bone,Name="spine_05")
    RotationOffset=(X=7.0096,Y=-141.722,Z=16.681)
    ExecutePin=@RigVMFunction_Sequence.A
   )
   > &RigUnit_ModifyTransforms(
    ItemToModify="((Item=(Type=Null,Name=\"def_bandolier_rear_01_ctrl_null\"),Transform=(Rotation=(X=0.1332,Y=0,Z=0,W=0.9911),Translation=(X=1.7248,Y=0.9874,Z=-4),Scale3D=(X=1,Y=1,Z=1))),(Item=(Type=Null,Name=\"def_bandolier_rear_02_ctrl_null\"),Transform=(Rotation=(X=0,Y=0,Z=0 ...(truncated)"
    WeightMinimum=0
    Weight=@n2.OutputParam
   )
   > n3 = &RigUnit_SphericalPoseReader(
    ActiveRegionScaleFactors=(PositiveWidth=2.5,NegativeWidth=3,PositiveHeight=3,NegativeHeight=3)
    ActiveRegionSize=0.1
    Debug=(bDrawDebug=false,DebugScale=25,DebugSegments=20,DebugThickness=0.25)
    DriverAxis=(X=1,Y=0,Z=0)
    DriverItem=(Type=Bone,Name="upperarm_l")
    FalloffRegionScaleFactors=(PositiveWidth=1,NegativeWidth=1,PositiveHeight=1,NegativeHeight=1)
    FalloffSize=0.5
    OptionalParentItem=(Type=Bone,Name="clavicle_l")
    RotationOffset=(X=0,Y=0,Z=0)
   )
   > &RigUnit_ModifyTransforms(
    ItemToModify="((Item=(Type=Null,Name=\"def_bandolier_front_01_ctrl_null\"),Transform=(Rotation=(X=0.4876,Y=0,Z=0,W=0.8731),Translation=(X=-0.9842,Y=3.0732,Z=5.3101),Scale3D=(X=1,Y=1,Z=1))),(Item=(Type=Null,Name=\"def_bandolier_rear_01_ctrl_null\"),Transform=(Rotation=(X=-0. ...(truncated)"
    WeightMinimum=0
    Weight=@n3.OutputParam
   )
   > n4 = &RigUnit_SphericalPoseReader(
    ActiveRegionScaleFactors=(PositiveWidth=0.5,NegativeWidth=3,PositiveHeight=1.5,NegativeHeight=1.5)
    ActiveRegionSize=0.1
    Debug=(bDrawDebug=false,DebugScale=25,DebugSegments=20,DebugThickness=0.25)
    DriverAxis=(X=1,Y=0,Z=0)
    DriverItem=(Type=Bone,Name="upperarm_l")
    FalloffRegionScaleFactors=(PositiveWidth=1,NegativeWidth=1,PositiveHeight=1,NegativeHeight=1)
    FalloffSize=0.4
    OptionalParentItem=(Type=Bone,Name="clavicle_l")
    RotationOffset=(X=90,Y=0,Z=0)
   )
   > &RigUnit_ModifyTransforms(
    ItemToModify="((Item=(Type=Null,Name=\"def_bandolier_rear_01_ctrl_null\"),Transform=(Rotation=(X=0.2865,Y=0.1023,Z=0.0697,W=0.9501),Translation=(X=0.0501,Y=1.3975,Z=-3.7365),Scale3D=(X=1,Y=1,Z=1))),(Item=(Type=Bone,Name=\"base_bandolier\"),Transform=(Rotation=(X=-0.1505,Y=- ...(truncated)"
    WeightMinimum=0
    Weight=@n4.OutputParam
   )
   > &RigUnit_ParentConstraint(
    AdvancedSettings=(InterpolationType=Shortest,RotationOrderForFilter=XYZ)
    Child=(Type=Null,Name="dyn_bag_sim_chain_03_ctrl_null")
    Filter=(TranslationFilter=(bX=True,bY=True,bZ=True),RotationFilter=(bX=true,bY=true,bZ=true),ScaleFilter=(bX=True,bY=True,bZ=True))
    Parents=((Item=(Type=Bone,Name="thigh_r"),Weight=1),(Item=(Type=Bone,Name="pelvis"),Weight=1))
    Weight=1
   )
   > &RigUnit_ParentConstraint(
    AdvancedSettings=(InterpolationType=Average,RotationOrderForFilter=XZY)
    Child=(Type=Null,Name="def_bandolier_front_10_ctrl_null")
    Parents=((Item=(Type=Control,Name="dyn_bag_sim_chain_03_ctrl"),Weight=1))
    Weight=1
   )
   > &RigUnit_ParentConstraint(
    AdvancedSettings=(InterpolationType=Average,RotationOrderForFilter=XZY)
    Child=(Type=Null,Name="def_bandolier_rear_10_ctrl_null")
    Parents=((Item=(Type=Control,Name="dyn_bag_sim_chain_03_ctrl"),Weight=1),(Item=(Type=Bone,Name="None"),Weight=1))
    Weight=1
   )
   > n5 = &RigUnit_SphericalPoseReader(
    ActiveRegionScaleFactors=(PositiveWidth=3.5,NegativeWidth=0.5,PositiveHeight=3,NegativeHeight=3)
    ActiveRegionSize=0.1
    Debug=(bDrawDebug=false,DebugScale=25,DebugSegments=20,DebugThickness=0.25)
    DriverAxis=(X=1,Y=0,Z=0)
    DriverItem=(Type=Bone,Name="thigh_r")
    FalloffRegionScaleFactors=(PositiveWidth=1,NegativeWidth=1,PositiveHeight=1,NegativeHeight=1)
    FalloffSize=0.45
    FlipHeightScaling=false
    FlipWidthScaling=true
    OptionalParentItem=(Type=Bone)
    RotationOffset=(X=-90,Y=0,Z=0)
   )
   > &RigUnit_ModifyTransforms(
    ItemToModify="((Item=(Type=Null,Name=\"dyn_bag_sim_chain_03_ctrl_null\"),Transform=(Rotation=(X=0,Y=0,Z=-0.0058,W=1),Translation=(X=-7.2087,Y=0,Z=0.9173),Scale3D=(X=1,Y=1,Z=1))),(Item=(Type=Null,Name=\"def_bandolier_front_10_ctrl_null\"),Transform=(Rotation=(X=0,Y=0,Z=-0.12 ...(truncated)"
    WeightMinimum=0
    Weight=@n5.OutputParam
   )
   > &RigUnit_SphericalPoseReader(
    ActiveRegionScaleFactors=(PositiveWidth=1,NegativeWidth=1,PositiveHeight=4,NegativeHeight=4)
    ActiveRegionSize=0.1
    Debug=(bDrawDebug=false,bDrawLocalAxes=False,DebugScale=25,DebugSegments=20,DebugThickness=0.25)
    DriverAxis=(X=-1,Y=0,Z=0)
    DriverItem=(Type=Bone,Name="thigh_r")
    FalloffRegionScaleFactors=(PositiveWidth=0.6,NegativeWidth=1,PositiveHeight=2,NegativeHeight=4)
    FalloffSize=0.45
    OptionalParentItem=(Type=Bone)
    RotationOffset=(X=20,Y=140,Z=-25)
   )
   > SplineIK = RigVMFunctionReferenceNode(
    Primary Axis=(X=1,Y=0,Z=0)
    Samples Per Segment=3
    Secondary Spline Direction=(X=0,Y=-5,Z=0)
    Spline Mode=BSpline
    Stretch=true
    Up Axis=(X=0,Y=-1,Z=0)
    debug=false
    Controls=RigVMVariableNode() {variable=bandolier_rear_ctrl_NORBAN}.Value
    Bones=RigVMVariableNode() {variable="bandolier_rear_joints_NORBAN "}.Value
   )
   > SplineIK_1 = RigVMFunctionReferenceNode(
    Primary Axis=(X=1,Y=0,Z=0)
    Samples Per Segment=3
    Secondary Spline Direction=(X=0,Y=1,Z=0)
    Spline Mode=BSpline
    Stretch=true
    Up Axis=(X=0,Y=1,Z=0)
    debug=false
    Controls=RigVMVariableNode() {variable=bandolier_front_ctrl_NORBAN}.Value
    Bones=RigVMVariableNode() {variable=bandolier_front_joints_NORBAN}.Value
   )
   > &RigUnit_ParentConstraint(
    AdvancedSettings=(InterpolationType=Average,RotationOrderForFilter=XZY)
    Child=(Type=Bone,Name="dyn_bag_front_01")
    Parents=((Item=(Type=Control,Name="dyn_bag_sim_chain_03_ctrl"),Weight=0.1))
    Weight=1
   )
   > &RigUnit_ParentConstraint(
    AdvancedSettings=(InterpolationType=Average,RotationOrderForFilter=XZY)
    Child=(Type=Bone,Name="dyn_bag_rear_01")
    Parents=((Item=(Type=Control,Name="dyn_bag_sim_chain_03_ctrl"),Weight=1))
    Weight=1
   )
   RigVMFunctionReturnNode(ExecuteContext=@RigVMFunction_Sequence.B)
  ]}
 New Function_ContainedGraph {nodes [
   RigVMFunctionEntryNode()
   > n1 = &RigUnit_SphericalPoseReader(
    ActiveRegionScaleFactors=(PositiveWidth=0.5,NegativeWidth=2,PositiveHeight=4.5,NegativeHeight=3)
    ActiveRegionSize=0.1
    Debug=(bDrawDebug=false,DebugScale=25,DebugSegments=20,DebugThickness=0.25)
    DriverAxis=(X=1,Y=0,Z=0)
    DriverItem=(Type=Bone,Name="thigh_r")
    FalloffRegionScaleFactors=(PositiveWidth=0.1,NegativeWidth=1,PositiveHeight=1,NegativeHeight=1)
    FalloffSize=0.45
    OptionalParentItem=(Type=Bone)
    RotationOffset=(X=90,Y=0,Z=-50)
   )
   > &RigUnit_ModifyTransforms(
    ItemToModify="((Item=(Type=Bone,Name=\"dyn_bandolier_rear_05\"),Transform=(Rotation=(X=0.0076,Y=0.0868,Z=-0.0868,W=0.9924),Translation=(X=0,Y=0,Z=0),Scale3D=(X=1,Y=1,Z=1))),(Item=(Type=Bone,Name=\"dyn_bag_sim_chain_01\"),Transform=(Rotation=(X=0,Y=-0.0872,Z=0,W=0.9962),Tran ...(truncated)"
    WeightMinimum=0
    Weight=@n1.OutputParam
   )
   > n2 = &RigUnit_SphericalPoseReader(
    ActiveRegionScaleFactors=(PositiveWidth=0.5,NegativeWidth=2,PositiveHeight=4.5,NegativeHeight=3)
    ActiveRegionSize=0.1
    Debug=(bDrawDebug=false,bDraw2D=false,DebugScale=25,DebugSegments=20,DebugThickness=0.25)
    DriverAxis=(X=1,Y=0,Z=0)
    DriverItem=(Type=Bone,Name="thigh_r")
    FalloffRegionScaleFactors=(PositiveWidth=0.1,NegativeWidth=1,PositiveHeight=1,NegativeHeight=1)
    FalloffSize=0.45
    OptionalParentItem=(Type=Bone)
    RotationOffset=(X=-60,Y=0,Z=10)
   )
   > n3 = &RigUnit_SphericalPoseReader(
    ActiveRegionScaleFactors=(PositiveWidth=1,NegativeWidth=0.5,PositiveHeight=1,NegativeHeight=1)
    ActiveRegionSize=0.1
    Debug=(bDrawDebug=false,DebugScale=25,DebugSegments=20,DebugThickness=0.25)
    DriverAxis=(X=1,Y=0,Z=0)
    DriverItem=(Type=Bone,Name="pelvis")
    FalloffRegionScaleFactors=(PositiveWidth=0.1,NegativeWidth=0.1,PositiveHeight=0.2,NegativeHeight=0.2)
    FalloffSize=0.5
    FlipHeightScaling=false
    FlipWidthScaling=false
    OptionalParentItem=(Type=Bone)
    RotationOffset=(X=0,Y=-110,Z=0)
   )
   > n4 = &RigUnit_SphericalPoseReader(
    ActiveRegionScaleFactors=(PositiveWidth=1,NegativeWidth=3.5,PositiveHeight=1.5,NegativeHeight=1.5)
    ActiveRegionSize=0.1
    Debug=(bDrawDebug=false,DebugScale=25,DebugSegments=20,DebugThickness=0.25)
    DriverAxis=(X=1,Y=0,Z=0)
    DriverItem=(Type=Bone,Name="thigh_r")
    FalloffRegionScaleFactors=(PositiveWidth=0.5,NegativeWidth=1,PositiveHeight=1,NegativeHeight=1)
    FalloffSize=0.35
    OptionalParentItem=(Type=Bone,Name="pelvis")
    RotationOffset=(X=-5,Y=-27,Z=0)
   )
   n5 = &RigVMFunction_AlphaInterp(
    InterpSpeedDecreasing=5
    InterpSpeedIncreasing=5
    Scale=1
    bClampResult=true
    bMapRange=true
    Value=RigVMVariableNode() {variable=Emote}.Value
   )
   > &RigUnit_ModifyTransforms(
    ItemToModify="((Item=(Type=Bone,Name=\"dyn_bandolier_front_05\"),Transform=(Rotation=(X=0,Y=-0.1736,Z=0,W=0.9848),Translation=(X=0,Y=0,Z=0),Scale3D=(X=1,Y=1,Z=1))),(Item=(Type=Bone,Name=\"dyn_bag_sim_chain_01\"),Transform=(Rotation=(X=0,Y=0.2164,Z=0,W=0.9763),Translation=(X ...(truncated)"
    WeightMinimum=0
    Weight=&RigVMFunction_MathDoubleMul(B=@n5.Result, A=@n4.OutputParam).Result
   )
   n6 = &RigVMFunction_AlphaInterp(
    InterpSpeedDecreasing=5
    InterpSpeedIncreasing=5
    Scale=1
    bClampResult=true
    bMapRange=true
    Value=RigVMVariableNode() {variable=Emote}.Value
   )
   > &RigUnit_ModifyTransforms(
    ItemToModify="((Item=(Type=Bone,Name=\"dyn_bag_sim_chain_01\"),Transform=(Rotation=(X=0,Y=-0.0872,Z=0,W=0.9962),Translation=(X=-2,Y=0,Z=0),Scale3D=(X=1,Y=1,Z=1))),(Item=(Type=Bone,Name=\"dyn_bandolier_front_05\"),Transform=(Rotation=(X=0,Y=0.1736,Z=0,W=0.9848),Translation=( ...(truncated)"
    WeightMinimum=0
    Weight=&RigVMFunction_MathDoubleMul(B=@n6.Result, A=@n3.OutputParam).Result
   )
   n7 = &RigVMFunction_AlphaInterp(
    InterpSpeedDecreasing=5
    InterpSpeedIncreasing=5
    Scale=1
    bClampResult=true
    bMapRange=true
    Value=RigVMVariableNode() {variable=Emote}.Value
   )
   > &RigUnit_ModifyTransforms(
    ItemToModify=((Item=(Type=Bone,Name="dyn_bandolier_front_05"),Transform=(Rotation=(X=0.1981,Y=0.1387,Z=-0.1981,W=0.9499),Translation=(X=0,Y=0,Z=0),Scale3D=(X=1,Y=1,Z=1))))
    WeightMinimum=0
    Weight=&RigVMFunction_MathDoubleMul(B=@n7.Result, A=@n2.OutputParam).Result
   )
   n8 = &RigVMFunction_AlphaInterp(
    InterpSpeedDecreasing=5
    InterpSpeedIncreasing=5
    Scale=1
    bClampResult=true
    bMapRange=true
    Value=&RigVMFunction_MathDoubleMul(B=RigVMVariableNode() {variable=Spine_Rot_L}.Value, A=RigVMVariableNode() {variable=Emote}.Value).Result
   )
   > &RigUnit_ModifyTransforms(
    ItemToModify="((Item=(Type=Bone,Name=\"dyn_bandolier_front_05\"),Transform=(Rotation=(X=0,Y=-0.0436,Z=0,W=0.999),Translation=(X=0,Y=0,Z=0),Scale3D=(X=1,Y=1,Z=1))),(Item=(Type=Bone,Name=\"dyn_bag_sim_chain_03\"),Transform=(Rotation=(X=0,Y=0,Z=0,W=1),Translation=(X=0,Y=-5,Z=0 ...(truncated)"
    WeightMinimum=0
    Weight=@n8.Result
   )
   > n9 = &RigUnit_SphericalPoseReader(
    ActiveRegionScaleFactors=(PositiveWidth=1,NegativeWidth=1,PositiveHeight=1,NegativeHeight=1)
    ActiveRegionSize=0.1
    Debug=(bDrawDebug=false,DebugScale=25,DebugSegments=20,DebugThickness=0.25)
    DriverAxis=(X=1,Y=0,Z=0)
    DriverItem=(Type=Bone,Name="upperarm_l")
    FalloffRegionScaleFactors=(PositiveWidth=1,NegativeWidth=1,PositiveHeight=1,NegativeHeight=1)
    FalloffSize=0.35
    OptionalParentItem=(Type=Bone,Name="spine_05")
    RotationOffset=(X=7.0096,Y=-141.722,Z=16.681)
   )
   n10 = &RigVMFunction_AlphaInterp(
    InterpSpeedDecreasing=10
    InterpSpeedIncreasing=5
    Scale=1
    bClampResult=true
    bMapRange=true
    Value=&RigVMFunction_MathDoubleMul(A=@n9.OutputParam, B=RigVMVariableNode() {variable=Emote}.Value).Result
   )
   > &RigUnit_ModifyTransforms(
    ItemToModify="((Item=(Type=Bone,Name=\"dyn_bag_sim_chain_01\"),Transform=(Rotation=(X=0,Y=-0.1736,Z=0,W=0.9848),Translation=(X=0,Y=0,Z=0),Scale3D=(X=1,Y=1,Z=1))),(Item=(Type=Bone,Name=\"dyn_bandolier_rear_06\"),Transform=(Rotation=(X=-0.2549,Y=0.1677,Z=0.0449,W=0.9513),Tran ...(truncated)"
    WeightMinimum=0
    Weight=@n10.Result
   )
   > RigVMFunctionReturnNode()
  ]}
 New Function_ContainedGraph {nodes [
   RigVMFunctionEntryNode()
   > &RigUnit_ParentConstraint(
    AdvancedSettings=(RotationOrderForFilter=XZY)
    Child=(Type=Null,Name="def_bandolier_rear_06_ctrl_null")
    Parents="((Item=(Type=Bone,Name=\"clavicle_r\"),Weight=0.0044),(Item=(Type=Bone,Name=\"pelvis\"),Weight=0.0093),(Item=(Type=Bone,Name=\"spine_05\"),Weight=0.0169),(Item=(Type=Bone,Name=\"spine_01\"),Weight=0.1165),(Item=(Type=Bone,Name=\"spine_04\"),Weight=0.2461),(Item=(Typ ...(truncated)"
    Weight=1
   )
   > &RigUnit_ParentConstraint(
    AdvancedSettings=(RotationOrderForFilter=XZY)
    Child=(Type=Null,Name="def_bandolier_rear_07_ctrl_null")
    Parents="((Item=(Type=Bone,Name=\"pelvis\"),Weight=0.0044),(Item=(Type=Bone,Name=\"spine_05\"),Weight=0.007),(Item=(Type=Bone,Name=\"clavicle_r\"),Weight=0.0076),(Item=(Type=Bone,Name=\"spine_01\"),Weight=0.135),(Item=(Type=Bone,Name=\"spine_04\"),Weight=0.233),(Item=(Type=B ...(truncated)"
    Weight=1
   )
   > &RigUnit_ParentConstraint(
    AdvancedSettings=(RotationOrderForFilter=XZY)
    Child=(Type=Null,Name="def_bandolier_rear_08_ctrl_null")
    Parents="((Item=(Type=Bone,Name=\"thigh_twist_01_r\"),Weight=0.0001),(Item=(Type=Bone,Name=\"deform_groin_r\"),Weight=0.0021),(Item=(Type=Bone,Name=\"thigh_r\"),Weight=0.0063),(Item=(Type=Bone,Name=\"deform_glute_l\"),Weight=0.0176),(Item=(Type=Bone,Name=\"spine_02\"),Weight ...(truncated)"
    Weight=1
   )
   > &RigUnit_ParentConstraint(
    AdvancedSettings=(RotationOrderForFilter=XZY)
    Child=(Type=Null,Name="def_bandolier_rear_09_ctrl_null")
    Parents="((Item=(Type=Bone,Name=\"thigh_twist_01_r\"),Weight=0.0011),(Item=(Type=Bone,Name=\"spine_02\"),Weight=0.017),(Item=(Type=Bone,Name=\"spine_01\"),Weight=0.0241),(Item=(Type=Bone,Name=\"deform_groin_r\"),Weight=0.0283),(Item=(Type=Bone,Name=\"deform_glute_l\"),Weight ...(truncated)"
    Weight=1
   )
   > &RigUnit_ParentConstraint(
    AdvancedSettings=(RotationOrderForFilter=XZY)
    Child=(Type=Null,Name="def_bandolier_rear_10_ctrl_null")
    Parents="((Item=(Type=Bone,Name=\"spine_02\"),Weight=0.0048),(Item=(Type=Bone,Name=\"thigh_twist_01_r\"),Weight=0.0048),(Item=(Type=Bone,Name=\"spine_01\"),Weight=0.0097),(Item=(Type=Bone,Name=\"deform_groin_r\"),Weight=0.0389),(Item=(Type=Bone,Name=\"deform_glute_l\"),Weigh ...(truncated)"
    Weight=1
   )
   > &RigUnit_ParentConstraint(
    AdvancedSettings=(RotationOrderForFilter=XZY)
    Child=(Type=Null,Name="def_bandolier_front_06_ctrl_null")
    Parents="((Item=(Type=Bone,Name=\"pelvis\"),Weight=0.0003),(Item=(Type=Bone,Name=\"deform_pec_aim_r\"),Weight=0.0009),(Item=(Type=Bone,Name=\"spine_01\"),Weight=0.0161),(Item=(Type=Bone,Name=\"spine_02\"),Weight=0.0573),(Item=(Type=Bone,Name=\"spine_03\"),Weight=0.0701),(Ite ...(truncated)"
    Weight=1
   )
   > &RigUnit_ParentConstraint(
    AdvancedSettings=(RotationOrderForFilter=XZY)
    Child=(Type=Null,Name="def_bandolier_front_07_ctrl_null")
    Parents="((Item=(Type=Bone,Name=\"deform_pec_aim_r\"),Weight=0.0036),(Item=(Type=Bone,Name=\"deform_groin_l\"),Weight=0.0103),(Item=(Type=Bone,Name=\"deform_groin_r\"),Weight=0.0142),(Item=(Type=Bone,Name=\"pelvis\"),Weight=0.1054),(Item=(Type=Bone,Name=\"spine_03\"),Weight= ...(truncated)"
    Weight=1
   )
   > &RigUnit_ParentConstraint(
    AdvancedSettings=(RotationOrderForFilter=XZY)
    Child=(Type=Null,Name="def_bandolier_front_08_ctrl_null")
    Parents="((Item=(Type=Bone,Name=\"thigh_r\"),Weight=0.0152),(Item=(Type=Bone,Name=\"spine_03\"),Weight=0.0379),(Item=(Type=Bone,Name=\"deform_groin_l\"),Weight=0.0389),(Item=(Type=Bone,Name=\"spine_04\"),Weight=0.0906),(Item=(Type=Bone,Name=\"spine_01\"),Weight=0.1638),(Item ...(truncated)"
    Weight=1
   )
   > &RigUnit_ParentConstraint(
    AdvancedSettings=(RotationOrderForFilter=XZY)
    Child=(Type=Null,Name="def_bandolier_front_09_ctrl_null")
    Parents="((Item=(Type=Bone,Name=\"spine_03\"),Weight=0.0029),(Item=(Type=Bone,Name=\"spine_04\"),Weight=0.0222),(Item=(Type=Bone,Name=\"deform_groin_l\"),Weight=0.0675),(Item=(Type=Bone,Name=\"spine_02\"),Weight=0.0725),(Item=(Type=Bone,Name=\"spine_01\"),Weight=0.0847),(Ite ...(truncated)"
    Weight=1
   )
   > &RigUnit_ParentConstraint(
    AdvancedSettings=(RotationOrderForFilter=XZY)
    Child=(Type=Null,Name="def_bandolier_front_10_ctrl_null")
    Parents="((Item=(Type=Bone,Name=\"thigh_twist_01_r\"),Weight=0.0002),(Item=(Type=Bone,Name=\"spine_04\"),Weight=0.0068),(Item=(Type=Bone,Name=\"deform_groin_l\"),Weight=0.022),(Item=(Type=Bone,Name=\"spine_02\"),Weight=0.0319),(Item=(Type=Bone,Name=\"spine_01\"),Weight=0.045 ...(truncated)"
    Weight=1
   )
   > RigVMFunctionReturnNode()
  ]}
 RigVMFunctionLibrary {nodes [
   Bandoleer_Solver = RigVMCollapseNode(ExecuteContext=())
   Bandoleer_Solver_NoRBAN = RigVMCollapseNode(ExecuteContext=())
   Control_container = RigVMCollapseNode(Debug=False, ExecuteContext=())
   Emote_Solver_FNC = RigVMCollapseNode(ExecuteContext=())
   Pin_FNC = RigVMCollapseNode(ExecuteContext=())
   Pin_NoRBAN_FNC = RigVMCollapseNode(ExecuteContext=())
   Spinal_Tracking_FNC = RigVMCollapseNode(ExecuteContext=())
  ]}
 RigVMFunction_Sequence_1_ContainedGraph {nodes [
   n1 = RigVMFunctionEntryNode()
   RigVMFunction_Sequence = &RigVMFunction_Sequence(ExecuteContext=@n1.ExecuteContext)
   n2 = &RigVMFunction_Sequence(ExecuteContext=@RigVMFunction_Sequence.B)
   RigVMFunctionReturnNode(A=@RigVMFunction_Sequence.A, B=@n2.A, C=@n2.B)
  ]}
 RigVMFunction_Sequence_2_ContainedGraph {nodes [
   n1 = RigVMFunctionEntryNode()
   n2 = &RigVMFunction_Sequence(ExecuteContext=@n1.ExecuteContext)
   n3 = &RigVMFunction_Sequence(ExecuteContext=@n2.B)
   RigVMFunctionReturnNode(A=@n2.A, B=@n3.A, C=@n3.B)
  ]}
 RigVMModel {nodes [
   RigUnit_BeginExecution = &RigUnit_BeginExecution()
   Control_container = RigVMFunctionReferenceNode(Debug=false, ExecuteContext=@RigUnit_BeginExecution.ExecutePin)
   > TwistsFix = RigVMCollapseNode()
   > FN_DeformRig_FNC = RigVMFunctionReferenceNode(
    DeformRig_ Curve="(deformRig_l_upperarm_up_alpha_72_#1=False,deformRig_r_upperarm_up_alpha_73_#2=False,deformRig_l_upperarm_fr_alpha_74_#3=False,deformRig_r_upperarm_fr_alpha_75_CB51B8 ...(truncated)"
    DeformRig_Bool="(upperarm_1_#4=True,elbow_3_#5=True,wrist_7_#6=True,groin_8_#7=True,glute_17_#8=true,knee_9_40C408324CA261 ...(truncated)"
    DeformRig_Debug=(upperarm_debug_18_#4=False,wrist_debug_20_#6=False,lat_debug_26_#9=False,groinGlute_debug_27_#7=False)
    DeformRig_Value="(upperarm_up_18_#10=(Rotation=(X=0,Y=0,Z=0,W=1),Translation=(X=0,Y=0,Z=6),Scale3D=(X=0.7,Y=1,Z=1)),upperarm_fwd_fr_19_#11=(Rotation=(X=0,Y=0,Z=0,W=1),Translation=(X=0,Y=10,Z=0),Scale3D=(X=0.5,Y=1,Z= ...(truncated)"
   )
   RigVMFunction_ControlFlowBranch = &RigVMFunction_ControlFlowBranch(ExecuteContext=@FN_DeformRig_FNC.ExecuteContext, Condition=RigVMVariableNode() {variable=RBAN}.Value)
   Pin_FNC_1 = RigVMFunctionReferenceNode(ExecuteContext=@RigVMFunction_ControlFlowBranch.False)
   > Pin_NoRBAN_FNC = RigVMFunctionReferenceNode()
   > Bandoleer_Solver_NoRBAN = RigVMFunctionReferenceNode()
   Pin_FNC = RigVMFunctionReferenceNode(ExecuteContext=@RigVMFunction_ControlFlowBranch.True)
   > Bandoleer_Solver = RigVMFunctionReferenceNode()
   > Spinal_Tracking_FNC = RigVMFunctionReferenceNode()
   > Emote_Solver_FNC = RigVMFunctionReferenceNode()
  ]}
]
hierarchy [
 root {loc=<0 0 0>, rot=<0 0 0 1>}
 pelvis {parent=root, loc=<0 -0.5301 98.0571>, rot=<0.0127 0.707 -0.0127 -0.707>}
 spine_01 {parent=pelvis, loc=<6.1208 -0.9328 0>, rot=<0 0 0.034 -0.9994>}
 spine_02 {parent=spine_01, loc=<6.5997 -0.2095 0>, rot=<0 0 0.0238 -0.9997>}
 spine_03 {parent=spine_02, loc=<6.4914 -0.1237 0>, rot=<0 0 0.0149 -0.9999>}
 spine_04 {parent=spine_03, loc=<6.9598 0.0794 0>, rot=<0 0 -0.0132 -0.9999>}
 spine_05 {parent=spine_04, loc=<6.9987 0 0>, rot=<0 0 -0.0015 -1>}
 neck_01 {parent=spine_05, loc=<11.3138 0.1529 0>, rot=<0 0 -0.2003 -0.9797>}
 neck_02 {parent=neck_01, loc=<4.5627 0 0>, rot=<0 0 0.0504 -0.9987>}
 head {parent=neck_02, loc=<5.3914 0 0>, rot=<0 0 0.115 -0.9934>}
 clavicle_l {parent=spine_05, loc=<5.4831 -0.1053 -3.2946>, rot=<0.0465 -0.7289 -0.009 -0.6829>}
 upperarm_l {parent=clavicle_l, loc=<9.8658 -1.2136 -0.4105>, rot=<0.0164 -0.3245 -0.0269 -0.9454>}
 lowerarm_l {parent=upperarm_l, loc=<26.465 -0.8345 0.259>, rot=<0.0036 0.0161 -0.2298 -0.9731>}
 hand_l {parent=lowerarm_l, loc=<22.7355 -0.3203 -0.6105>, rot=<-0.6784 -0.0223 0.0735 -0.7307>}
 thumb_01_l {parent=hand_l, loc=<1.3118 -0.3191 -2.459>, rot=<0.6546 -0.4042 0.0664 -0.6354>}
 thumb_02_l {parent=thumb_01_l, loc=<5.9336 -0.0448 0.0401>, rot=<-0.0037 0.0213 0.298 -0.9543>}
 thumb_03_l {parent=thumb_02_l, loc=<2.2223 0.3682 -0.0907>, rot=<0.0118 -0.013 0.0066 -0.9998>}
 index_metacarpal_l {parent=hand_l, loc=<1.6535 0.0183 -2.96>, rot=<0.0158 -0.0124 -0.0173 -0.9996>}
 index_01_l {parent=index_metacarpal_l, loc=<8.3008 0.0357 -0.0524>, rot=<-0.0044 0.0485 0.0924 -0.9945>}
 index_02_l {parent=index_01_l, loc=<3.3587 -0.3004 0.0452>, rot=<-0.002 0.0027 0.3225 -0.9466>}
 index_03_l {parent=index_02_l, loc=<2.5578 -0.123 -0.0252>, rot=<-0.002 0.0067 0.2558 -0.9667>}
 middle_metacarpal_l {parent=hand_l, loc=<2.1856 -0.1725 -0.5984>, rot=<0.006 0.0168 -0.0329 -0.9993>}
 middle_01_l {parent=middle_metacarpal_l, loc=<7.3275 -0.2447 0.0129>, rot=<-0.0208 0.0042 0.0858 -0.9961>}
 middle_02_l {parent=middle_01_l, loc=<4.0123 0.1401 0.1175>, rot=<-0.0218 0.0258 0.3512 -0.9357>}
 middle_03_l {parent=middle_02_l, loc=<2.6077 0.0534 0.1373>, rot=<0.0082 0.0188 0.1708 -0.9851>}
 ring_metacarpal_l {parent=hand_l, loc=<2.2074 -0.3625 1.3638>, rot=<-0.0284 0.0722 -0.0458 -0.9959>}
 ring_01_l {parent=ring_metacarpal_l, loc=<7.0295 -0.6197 -0.1446>, rot=<-0.0489 -0.0252 0.1892 -0.9804>}
 ring_02_l {parent=ring_01_l, loc=<3.6174 0.4647 -0.1027>, rot=<0 0 0.2806 -0.9598>}
 ring_03_l {parent=ring_02_l, loc=<2.4152 0.114 0.0199>, rot=<0 0 0.1189 -0.9929>}
 pinky_metacarpal_l {parent=hand_l, loc=<1.7616 -0.9753 3.3814>, rot=<-0.0397 0.0814 -0.069 -0.9935>}
 pinky_01_l {parent=pinky_metacarpal_l, loc=<7.1709 -0.4831 -0.1094>, rot=<-0.0654 -0.0035 0.2242 -0.9723>}
 pinky_02_l {parent=pinky_01_l, loc=<2.8274 0.0893 -0.1043>, rot=<0.0037 0.0011 0.2854 -0.9584>}
 pinky_03_l {parent=pinky_02_l, loc=<2.2046 0.0307 -0.0078>, rot=<-0.0002 0.0098 0.1445 -0.9895>}
 deform_wrist_dn_base_l {parent=hand_l, loc=<0 -1.5326 0>, rot=<0 0 0 -1>}
 deform_wrist_dn_l {parent=deform_wrist_dn_base_l, loc=<0 0 0>, rot=<0 0 0 1>}
 deform_wrist_up_base_l {parent=hand_l, loc=<0 2.3009 0>, rot=<0 0 0 -1>}
 deform_wrist_up_l {parent=deform_wrist_up_base_l, loc=<0 0 0>, rot=<0 0 0 1>}
 lowerarm_twist_02_l {parent=lowerarm_l, loc=<10.6707 0 0>, rot=<0 0 0 1>}
 lowerarm_twist_01_l {parent=lowerarm_l, loc=<18.316 0 0>, rot=<0 0 0 1>}
 deform_elbow_in_base_l {parent=lowerarm_l, loc=<0.1834 3.6062 -0.2572>, rot=<0 0 0 -1>}
 deform_elbow_in_l {parent=deform_elbow_in_base_l, loc=<0 0 0>, rot=<0 0 0 1>}
 deform_elbow_out_base_l {parent=lowerarm_l, loc=<0 -4 0>, rot=<0 0 0 1>}
 deform_elbow_out_l {parent=deform_elbow_out_base_l, loc=<0 0 0>, rot=<0 0 0 1>}
 upperarm_twist_01_l {parent=upperarm_l, loc=<7.2135 -0.1134 0>, rot=<0 0 0 1>}
 deform_upperarm_bk_base_l {parent=upperarm_twist_01_l, loc=<-6.4053 -4.3746 0.314>, rot=<0 0 0 -1>}
 deform_upperarm_bk_l {parent=deform_upperarm_bk_base_l, loc=<0 0 0>, rot=<0 0 0 1>}
 deform_upperarm_fr_base_l {parent=upperarm_twist_01_l, loc=<-5.6312 5.092 0.3354>, rot=<0 0 0 -1>}
 deform_upperarm_fr_l {parent=deform_upperarm_fr_base_l, loc=<0 0 0>, rot=<0 0 0 1>}
 deform_upperarm_up_base_l {parent=upperarm_twist_01_l, loc=<-8.891 0.4825 3.9517>, rot=<0 0 0 -1>}
 deform_upperarm_up_l {parent=deform_upperarm_up_base_l, loc=<0 0 0>, rot=<0 0 0 1>}
 upperarm_twist_02_l {parent=upperarm_l, loc=<19.7998 -0.1134 0>, rot=<0 0 0 1>}
 clavicle_r {parent=spine_05, loc=<5.4831 -0.1053 3.2946>, rot=<0.7289 0.0465 -0.6829 0.009>}
 upperarm_r {parent=clavicle_r, loc=<-9.8658 1.2136 0.4105>, rot=<0.0164 -0.3245 -0.0269 -0.9454>}
 lowerarm_r {parent=upperarm_r, loc=<-26.465 0.8345 -0.259>, rot=<0.0036 0.0161 -0.2298 -0.9731>}
 hand_r {parent=lowerarm_r, loc=<-22.7355 0.3203 0.6106>, rot=<-0.6784 -0.0223 0.0735 -0.7307>}
 thumb_01_r {parent=hand_r, loc=<-1.3118 0.3191 2.459>, rot=<0.6546 -0.4042 0.0664 -0.6354>}
 thumb_02_r {parent=thumb_01_r, loc=<-5.9336 0.0448 -0.0401>, rot=<-0.0037 0.0213 0.298 -0.9543>}
 thumb_03_r {parent=thumb_02_r, loc=<-2.2223 -0.3682 0.0907>, rot=<0.0118 -0.013 0.0066 -0.9998>}
 index_metacarpal_r {parent=hand_r, loc=<-1.6535 -0.0183 2.96>, rot=<0.0158 -0.0124 -0.0173 -0.9996>}
 index_01_r {parent=index_metacarpal_r, loc=<-8.3008 -0.0357 0.0524>, rot=<-0.0044 0.0485 0.0924 -0.9945>}
 index_02_r {parent=index_01_r, loc=<-3.3587 0.3004 -0.0452>, rot=<-0.002 0.0027 0.3225 -0.9466>}
 index_03_r {parent=index_02_r, loc=<-2.5578 0.123 0.0252>, rot=<-0.002 0.0067 0.2558 -0.9667>}
 middle_metacarpal_r {parent=hand_r, loc=<-2.1856 0.1725 0.5984>, rot=<0.006 0.0168 -0.0329 -0.9993>}
 middle_01_r {parent=middle_metacarpal_r, loc=<-7.3275 0.2447 -0.0129>, rot=<-0.0208 0.0042 0.0858 -0.9961>}
 middle_02_r {parent=middle_01_r, loc=<-4.0123 -0.1401 -0.1175>, rot=<-0.0218 0.0258 0.3512 -0.9357>}
 middle_03_r {parent=middle_02_r, loc=<-2.6077 -0.0534 -0.1373>, rot=<0.0082 0.0188 0.1708 -0.9851>}
 ring_metacarpal_r {parent=hand_r, loc=<-2.2074 0.3625 -1.3638>, rot=<-0.0284 0.0722 -0.0458 -0.9959>}
 ring_01_r {parent=ring_metacarpal_r, loc=<-7.0295 0.6197 0.1446>, rot=<-0.0489 -0.0252 0.1892 -0.9804>}
 ring_02_r {parent=ring_01_r, loc=<-3.6174 -0.4647 0.1027>, rot=<0 0 0.2806 -0.9598>}
 ring_03_r {parent=ring_02_r, loc=<-2.4152 -0.114 -0.0199>, rot=<0 0 0.1189 -0.9929>}
 pinky_metacarpal_r {parent=hand_r, loc=<-1.7616 0.9753 -3.3814>, rot=<-0.0397 0.0814 -0.069 -0.9935>}
 pinky_01_r {parent=pinky_metacarpal_r, loc=<-7.1709 0.4831 0.1094>, rot=<-0.0654 -0.0035 0.2242 -0.9723>}
 pinky_02_r {parent=pinky_01_r, loc=<-2.8274 -0.0893 0.1043>, rot=<0.0037 0.0011 0.2854 -0.9584>}
 pinky_03_r {parent=pinky_02_r, loc=<-2.2046 -0.0307 0.0078>, rot=<-0.0002 0.0098 0.1445 -0.9895>}
 deform_wrist_dn_base_r {parent=hand_r, loc=<-0.0001 1.5328 0>, rot=<0 0 0 -1>}
 deform_wrist_dn_r {parent=deform_wrist_dn_base_r, loc=<0 0 0>, rot=<0 0 0 1>}
 deform_wrist_up_base_r {parent=hand_r, loc=<-0.0002 -2.3007 -0.0001>, rot=<0 0 0 -1>}
 deform_wrist_up_r {parent=deform_wrist_up_base_r, loc=<0 0 0>, rot=<0 0 0 1>}
 lowerarm_twist_02_r {parent=lowerarm_r, loc=<-10.6707 0 0>, rot=<0 0 0 -1>}
 lowerarm_twist_01_r {parent=lowerarm_r, loc=<-18.316 0 0>, rot=<0 0 0 -1>}
 deform_elbow_in_base_r {parent=lowerarm_r, loc=<-0.1834 -3.6062 0.2572>, rot=<0 0 0 -1>}
 deform_elbow_in_r {parent=deform_elbow_in_base_r, loc=<0 0 0>, rot=<0 0 0 1>}
 deform_elbow_out_base_r {parent=lowerarm_r, loc=<0 4 -0.0001>, rot=<0 0 0 -1>}
 deform_elbow_out_r {parent=deform_elbow_out_base_r, loc=<0 0 0>, rot=<0 0 0 1>}
 upperarm_twist_01_r {parent=upperarm_r, loc=<-7.2135 0.1134 0>, rot=<0 0 0 -1>}
 deform_upperarm_bk_base_r {parent=upperarm_twist_01_r, loc=<6.4053 4.3746 -0.314>, rot=<0 0 0 -1>}
 deform_upperarm_bk_r {parent=deform_upperarm_bk_base_r, loc=<0 0 0>, rot=<0 0 0 1>}
 deform_upperarm_fr_base_r {parent=upperarm_twist_01_r, loc=<5.6312 -5.092 -0.3354>, rot=<0 0 0 -1>}
 deform_upperarm_fr_r {parent=deform_upperarm_fr_base_r, loc=<0 0 0>, rot=<0 0 0 1>}
 deform_upperarm_up_base_r {parent=upperarm_twist_01_r, loc=<8.891 -0.4825 -3.9517>, rot=<0 0 0 -1>}
 deform_upperarm_up_r {parent=deform_upperarm_up_base_r, loc=<0 0 0>, rot=<0 0 0 1>}
 base_bandana {parent=upperarm_twist_01_r, loc=<-6.8149 5.9089 0.8799>, rot=<0 0 0 1>}
 dyn_bandana {parent=base_bandana, loc=<0 0 0>, rot=<0 0 0 1>}
 upperarm_twist_02_r {parent=upperarm_r, loc=<-19.7998 0.1134 0>, rot=<0 0 0 -1>}
 attach_backpack {parent=spine_05, loc=<-0.1166 -13.4166 0.0009>, rot=<-0.0093 -0.707 0.0093 -0.707>}
 Backpack_BR {parent=attach_backpack, type=Null, loc=<0 2.3256 0>, rot=<0 0 0 1>}
 attach_cape {parent=spine_05, loc=<9.0217 -11.6692 0.0045>, rot=<0.0616 -0.7044 -0.0616 -0.7044>}
 cape {parent=attach_cape, type=Null, loc=<0 2.5607 0>, rot=<0 0 0 1>}
 base_bandolier {parent=spine_05, loc=<10.0493 -1.3108 -12.4955>, rot=<0.0284 -0.7065 -0.0284 -0.7065>}
 dyn_bandolier {parent=base_bandolier, loc=<0.9979 0 0.3997>, rot=<0 0 0 1>}
 dyn_bandolier_rear_01 {parent=dyn_bandolier, loc=<-0.1001 -3.6926 -0.4954>, rot=<0.3526 -0.7646 0.2259 -0.4899>}
 dyn_bandolier_rear_02 {parent=dyn_bandolier_rear_01, loc=<6.3761 0.5626 -0.056>, rot=<0.0077 -0.0113 -0.264 -0.9644>}
 dyn_bandolier_rear_03 {parent=dyn_bandolier_rear_02, loc=<5.8957 0.209 -0.4111>, rot=<0.0014 -0.0056 -0.0843 -0.9964>}
 dyn_bandolier_rear_04 {parent=dyn_bandolier_rear_03, loc=<6.1441 0.5332 0.2555>, rot=<0.0007 -0.0068 -0.0616 -0.9981>}
 dyn_bandolier_rear_05 {parent=dyn_bandolier_rear_04, loc=<6.0423 0.364 0.0345>, rot=<0.0001 -0.0095 -0.0313 -0.9995>}
 dyn_bandolier_rear_06 {parent=dyn_bandolier_rear_05, loc=<6.1009 0.3755 -0.2858>, rot=<-0.0012 -0.0252 -0.0213 -0.9995>}
 dyn_bandolier_rear_07 {parent=dyn_bandolier_rear_06, loc=<5.731 -0.161 0.2877>, rot=<-0.0017 -0.0175 -0.0294 -0.9994>}
 dyn_bandolier_rear_08 {parent=dyn_bandolier_rear_07, loc=<5.5226 0.2791 -0.4813>, rot=<-0.0015 -0.0098 -0.0263 -0.9996>}
 dyn_bandolier_rear_09 {parent=dyn_bandolier_rear_08, loc=<6.0728 0.0333 -0.5329>, rot=<-0.0658 -0.0061 -0.0423 -0.9969>}
 dyn_bandolier_rear_10 {parent=dyn_bandolier_rear_09, loc=<5.596 0.6014 -0.2022>, rot=<-0.2256 -0.1405 -0.1785 -0.9474>}
 dyn_bandolier_rear_strap {parent=dyn_bandolier_rear_10, loc=<5.2109 0 0>, rot=<0 0 0 1>}
 base_bag_rear_01 {parent=dyn_bandolier_rear_strap, loc=<5.457 -0.6304 -1.0162>, rot=<0.8345 -0.2747 0.3055 0.3673>}
 dyn_bag_rear_01 {parent=base_bag_rear_01, loc=<0 0 0>, rot=<0 0 0 1>}
 dyn_bag_rear_02 {parent=dyn_bag_rear_01, loc=<5.9852 1.0006 0.7443>, rot=<0.9736 0.0257 -0.1778 0.1405>}
 dyn_bag_rear_03 {parent=dyn_bag_rear_02, loc=<9.2792 -0.2801 -4.0163>, rot=<0.9736 0.0257 -0.1778 -0.1405>}
 dyn_bag_rear_04 {parent=dyn_bag_rear_03, loc=<2.4926 1.5484 -3.5142>, rot=<0 -0.067 0 -0.9978>}
 dyn_bag_rear_05 {parent=dyn_bag_rear_04, loc=<1.8773 0.1932 -1.3939>, rot=<0 0 0 1>}
 base_keychain {parent=dyn_bag_rear_01, loc=<-0.6937 -0.2816 6.3344>, rot=<0 0 0 1>}
 dyn_keychain {parent=base_keychain, loc=<0 0 0>, rot=<0 0 0 1>}
 dyn_keychain1 {parent=dyn_keychain, loc=<1.9935 0.0945 0.4172>, rot=<0 0 0 1>}
 dyn_keychain2 {parent=dyn_keychain1, loc=<1.7414 -0.3078 0.6838>, rot=<0 0 0 1>}
 dyn_keychain3 {parent=dyn_keychain2, loc=<1.4867 0.0521 0.4053>, rot=<0 0 0 1>}
 dyn_bandolier_front_01 {parent=dyn_bandolier, loc=<0.3334 3.0788 -0.3629>, rot=<-0.3565 -0.7608 -0.2301 -0.4911>}
 dyn_bandolier_front_02 {parent=dyn_bandolier_front_01, loc=<5.3857 0.5351 0.3927>, rot=<-0.0119 -0.0141 0.1726 -0.9848>}
 dyn_bandolier_front_03 {parent=dyn_bandolier_front_02, loc=<5.4009 1.3294 0.2336>, rot=<-0.0095 -0.0218 0.1179 -0.9927>}
 dyn_bandolier_front_04 {parent=dyn_bandolier_front_03, loc=<5.9833 2.6014 0.5028>, rot=<-0.0007 -0.003 0.0653 -0.9979>}
 dyn_bandolier_front_05 {parent=dyn_bandolier_front_04, loc=<5.7348 1.3258 0.3724>, rot=<-0.0002 -0.0015 0.0301 -0.9995>}
 dyn_bandolier_front_06 {parent=dyn_bandolier_front_05, loc=<8.1702 -0.5469 0>, rot=<-0.0009 -0.0232 0.0646 -0.9976>}
 dyn_bandolier_front_07 {parent=dyn_bandolier_front_06, loc=<6.7835 -0.1926 0>, rot=<0.0009 -0.0106 0.0615 -0.998>}
 dyn_bandolier_front_08 {parent=dyn_bandolier_front_07, loc=<6.3738 0.0216 -0.1587>, rot=<0.0042 -0.0213 0.044 -0.9988>}
 dyn_bandolier_front_09 {parent=dyn_bandolier_front_08, loc=<6.6311 0.0731 -0.4253>, rot=<0.0118 -0.0375 0.067 -0.997>}
 dyn_bandolier_front_10 {parent=dyn_bandolier_front_09, loc=<5.2127 0.2769 -0.4172>, rot=<0.2597 -0.0342 -0.0038 -0.9651>}
 dyn_bandolier_front_strap {parent=dyn_bandolier_front_10, loc=<4.4282 0 0>, rot=<0 0 0 1>}
 base_bag_front {parent=dyn_bandolier_front_strap, loc=<6.2443 0.3114 -0.6084>, rot=<0.9223 0.0982 0.2966 -0.2277>}
 dyn_bag_front_01 {parent=base_bag_front, loc=<0 0 0>, rot=<0 0 0 1>}
 dyn_bag_front_02 {parent=dyn_bag_front_01, loc=<6.1721 0.8835 0.1171>, rot=<0.96 0.0119 -0.2356 -0.151>}
 dyn_bag_front_03 {parent=dyn_bag_front_02, loc=<9.3074 -1.576 -5.0164>, rot=<0.9702 0.0192 -0.189 0.1503>}
 dyn_bag_front_04 {parent=dyn_bag_front_03, loc=<2.9465 1.3384 -3.2345>, rot=<-0.1258 0 0 -0.9921>}
 dyn_bag_front_05 {parent=dyn_bag_front_04, loc=<1.6583 0.0859 -1.3804>, rot=<0 0 0 1>}
 deform_pec_aim_base_l {parent=spine_04, loc=<3.2456 9.3309 -1.2237>, rot=<0.0603 -0.4033 0.3318 -0.8507>}
 deform_pec_aim_l {parent=deform_pec_aim_base_l, loc=<0 0 0>, rot=<0 0 0 1>}
 deform_pec_base_l {parent=deform_pec_aim_l, loc=<9.3665 0 0>, rot=<0 0 0 1>}
 deform_pec_l {parent=deform_pec_base_l, loc=<0 0 0>, rot=<0 0 0 1>}
 deform_pec_aim_base_r {parent=spine_04, loc=<3.2452 9.3309 1.2237>, rot=<-0.4033 -0.0603 0.8507 0.3318>}
 deform_pec_aim_r {parent=deform_pec_aim_base_r, loc=<0 0 0>, rot=<0 0 0 1>}
 deform_pec_base_r {parent=deform_pec_aim_r, loc=<-9.3667 -0.0003 -0.0004>, rot=<0 0 0 1>}
 deform_pec_r {parent=deform_pec_base_r, loc=<0 0 0>, rot=<0 0 0 1>}
 dyn_bag_sim_chain_01_MSTR {parent=spine_03, loc=<0.1248 0.9083 -0.3686>, rot=<0.2485 -0.0413 0.9677 -0.0102>}
 dyn_bag_sim_chain_01 {parent=dyn_bag_sim_chain_01_MSTR, loc=<0 0 0>, rot=<0 0 0 1>}
 dyn_bag_sim_chain_02 {parent=dyn_bag_sim_chain_01, loc=<14.0717 0.0014 -0.0198>, rot=<0 0.1146 0 -0.9934>}
 dyn_bag_sim_chain_03 {parent=dyn_bag_sim_chain_02, loc=<17.2789 -0.0618 1.5027>, rot=<0.0578 -0.1891 -0.0112 -0.9802>}
 thigh_l {parent=pelvis, loc=<-1.0855 -0.2109 -8.7053>, rot=<0.0009 0.0355 0.0155 -0.9993>}
 calf_l {parent=thigh_l, loc=<-42.5284 -0.0019 -0.7297>, rot=<0.0006 -0.0161 -0.0391 -0.9991>}
 foot_l {parent=calf_l, loc=<-46.7401 -2.5014 -0.592>, rot=<0.0427 -0.0238 0.0353 -0.9982>}
 ball_l {parent=foot_l, loc=<-3.5658 10.8066 -0.1544>, rot=<0.0001 -0.0001 -0.7063 -0.7079>}
 calf_twist_02_l {parent=calf_l, loc=<-19.1546 -1.7943 -0.2638>, rot=<0.0237 0.0095 -0.0035 -0.9997>}
 calf_twist_01_l {parent=calf_l, loc=<-35.8483 -1.9197 -0.5794>, rot=<0.0237 0.0095 -0.0035 -0.9997>}
 deform_knee_in_base_l {parent=calf_l, loc=<0.8031 -5.0231 0.2456>, rot=<0 0 0 -1>}
 deform_knee_in_l {parent=deform_knee_in_base_l, loc=<0 0 0>, rot=<0 0 0 1>}
 deform_knee_out_base_l {parent=calf_l, loc=<0 4 0>, rot=<0 0 0 -1>}
 deform_knee_out_l {parent=deform_knee_out_base_l, loc=<0 0 0>, rot=<0 0 0 1>}
 thigh_twist_01_l {parent=thigh_l, loc=<-27.952 0 0>, rot=<0.0013 0 0 -1>}
 deform_groin_base_l {parent=thigh_l, loc=<0.4801 5.6776 -0.2486>, rot=<0 0 0 -1>}
 deform_groin_l {parent=deform_groin_base_l, loc=<0 0 0>, rot=<0 0 0 1>}
 thigh_r {parent=pelvis, loc=<-1.0855 -0.2109 8.7053>, rot=<0.0355 -0.0009 0.9993 0.0155>}
 calf_r {parent=thigh_r, loc=<42.5284 0.0019 0.7297>, rot=<0.0006 -0.0161 -0.0391 -0.9991>}
 foot_r {parent=calf_r, loc=<46.7401 2.5014 0.592>, rot=<0.0427 -0.0238 0.0353 -0.9982>}
 ball_r {parent=foot_r, loc=<3.5658 -10.8066 0.1544>, rot=<0.0001 -0.0001 -0.7063 -0.7079>}
 calf_twist_02_r {parent=calf_r, loc=<19.1546 1.7943 0.2638>, rot=<0.0237 0.0095 -0.0035 -0.9997>}
 base_lace {parent=calf_twist_02_r, loc=<2.8822 -3.5625 0.7084>, rot=<0 0 0 1>}
 dyn_lace {parent=base_lace, loc=<0.0069 -0.0019 -0.0335>, rot=<-0.01 -0.0861 0.1168 -0.9894>}
 dyn_lace_r {parent=dyn_lace, loc=<4.4893 0.4253 2.9724>, rot=<-0.0077 -0.0857 0.1244 -0.9885>}
 dyn_lace_l {parent=dyn_lace, loc=<5.8666 -0.3421 -1.7377>, rot=<-0.0115 -0.085 0.1132 -0.9899>}
 calf_twist_01_r {parent=calf_r, loc=<35.8483 1.9197 0.5794>, rot=<0.0237 0.0095 -0.0035 -0.9997>}
 deform_knee_in_base_r {parent=calf_r, loc=<-0.8031 5.0231 -0.2457>, rot=<0 0 0 -1>}
 deform_knee_in_r {parent=deform_knee_in_base_r, loc=<0 0 0>, rot=<0 0 0 1>}
 deform_knee_out_base_r {parent=calf_r, loc=<0 -4 0>, rot=<0 0 0 -1>}
 deform_knee_out_r {parent=deform_knee_out_base_r, loc=<0 0 0>, rot=<0 0 0 1>}
 thigh_twist_01_r {parent=thigh_r, loc=<27.952 0 0>, rot=<0.0013 0 0 -1>}
 deform_groin_base_r {parent=thigh_r, loc=<-0.4801 -5.6776 0.2486>, rot=<0 0 0 -1>}
 deform_groin_r {parent=deform_groin_base_r, loc=<0 0 0>, rot=<0 0 0 1>}
 thigh_r_null {parent=thigh_r, type=Null, loc=<0 0 0>, rot=<0 0 0 1>}
 deform_glute_base_l {parent=pelvis, loc=<-7.3466 -9.0213 -9.1245>, rot=<0.0172 0.031 -0.4477 -0.8935>}
 deform_glute_l {parent=deform_glute_base_l, loc=<0 0 0>, rot=<0 0 0 1>}
 deform_glute_base_r {parent=pelvis, loc=<-7.3464 -9.0213 9.1245>, rot=<0.031 -0.0172 0.8935 -0.4477>}
 deform_glute_r {parent=deform_glute_base_r, loc=<0 0 0>, rot=<0 0 0 1>}
 def_bandolier_ctrl_null {type=Null, loc=<12.4955 -5.6814 141.071>, rot=<0 0 0 1>}
 def_bandolier_ctrl {controlType=EULER_TRANSFORM, parent=def_bandolier_ctrl_null, type=Control, loc=<0 0 0>, rot=<0 0 0 1>}
 def_bandolier_rear_01_ctrl_null {type=Null, loc=<12.3954 -8.8462 140.5691>, rot=<0.3526 -0.7646 0.2259 -0.4899>}
 def_bandolier_rear_01_ctrl {controlType=EULER_TRANSFORM, parent=def_bandolier_rear_01_ctrl_null, type=Control, loc=<-82.29 -105.1335 -47.5894>, rot=<0 0 0 1>}
 def_bandolier_rear_02_ctrl_null {type=Null, loc=<10.436 -12.8215 136.9497>, rot=<0.1394 -0.8378 0.0866 -0.5208>}
 def_bandolier_rear_02_ctrl {
  controlType=EULER_TRANSFORM
  parent=def_bandolier_rear_02_ctrl_null
  type=Control
  loc=<-6.2754 -15.1334 136.9722>
  rot=<0.0918 0.8395 -0.2162 0.4899>
 }
 def_bandolier_rear_03_ctrl_null {type=Null, loc=<8.5123 -14.8634 131.5986>, rot=<0.0685 -0.8496 0.042 -0.5213>}
 def_bandolier_rear_03_ctrl {
  controlType=EULER_TRANSFORM
  parent=def_bandolier_rear_03_ctrl_null
  type=Control
  loc=<-1.5143 -15.2157 130.9187>
  rot=<0.0044 0.8494 -0.085 0.5208>
 }
 def_bandolier_rear_04_ctrl_null {type=Null, loc=<5.9543 -14.9472 125.9997>, rot=<-0.0161 0.8557 -0.0097 0.5171>}
 def_bandolier_rear_04_ctrl {
  controlType=EULER_TRANSFORM
  parent=def_bandolier_rear_04_ctrl_null
  type=Control
  loc=<-3.7979 -15.7778 125.696>
  rot=<-0.0371 -0.8516 0.0411 -0.5213>
 }
 def_bandolier_rear_05_ctrl_null {type=Null, loc=<3.7439 -14.3328 120.6264>, rot=<-0.0107 -0.8607 -0.0063 -0.509>}
 def_bandolier_rear_05_ctrl {
  controlType=EULER_TRANSFORM
  parent=def_bandolier_rear_05_ctrl_null
  type=Control
  loc=<-5.2324 -14.952 120.0134>
  rot=<0.0376 0.8551 -0.0094 0.5171>
 }
 def_bandolier_rear_06_ctrl_null {type=Null, loc=<1.0705 -13.8053 115.056>, rot=<-0.0295 -0.8728 -0.0164 -0.4869>}
 def_bandolier_rear_06_ctrl {
  controlType=EULER_TRANSFORM
  parent=def_bandolier_rear_06_ctrl_null
  type=Control
  loc=<-11.4667 -13.5595 114.3298>
  rot=<0.047 0.8595 0.0058 0.509>
 }
 def_bandolier_rear_07_ctrl_null {type=Null, loc=<-1.6942 -13.3419 110.2351>, rot=<-0.0556 -0.88 -0.0298 -0.4708>}
 def_bandolier_rear_07_ctrl {
  controlType=EULER_TRANSFORM
  parent=def_bandolier_rear_07_ctrl_null
  type=Control
  loc=<-11.881 -12.4928 109.7636>
  rot=<0.0801 0.8696 0.0154 0.4869>
 }
 def_bandolier_rear_08_ctrl_null {type=Null, loc=<-5.0591 -13.1734 105.6691>, rot=<-0.0792 -0.8828 -0.0414 -0.4611>}
 def_bandolier_rear_08_ctrl {
  controlType=EULER_TRANSFORM
  parent=def_bandolier_rear_08_ctrl_null
  type=Control
  loc=<-13.4084 -11.6742 105.7354>
  rot=<0.1013 0.8759 0.0287 0.4708>
 }
 def_bandolier_rear_09_ctrl_null {type=Null, loc=<-9.7459 -12.5431 101.2891>, rot=<-0.1169 -0.8802 -0.0606 -0.456>}
 def_bandolier_rear_09_ctrl {
  controlType=EULER_TRANSFORM
  parent=def_bandolier_rear_09_ctrl_null
  type=Control
  loc=<-17.674 -10.4561 102.5384>
  rot=<0.1538 0.8729 0.0408 0.4611>
 }
 def_bandolier_rear_10_ctrl_null {type=Null, loc=<-12.9367 -10.5016 97.1215>, rot=<-0.1265 -0.8049 0.2649 -0.5158>}
 def_bandolier_rear_10_ctrl {
  controlType=EULER_TRANSFORM
  parent=def_bandolier_rear_10_ctrl_null
  type=Control
  loc=<2.0428 42.1543 90.3188>
  rot=<-0.1667 0.7663 -0.4208 0.456>
 }
 def_bandolier_rear_strap_ctrl_null {type=Null, loc=<-15.9491 -7.5405 94.6588>, rot=<-0.248 -0.7164 0.4525 -0.4696>}
 def_bandolier_rear_strap_ctrl {
  controlType=EULER_TRANSFORM
  parent=def_bandolier_rear_strap_ctrl_null
  type=Control
  loc=<-19.2655 37.9074 90.6319>
  rot=<0.1265 0.585 -0.613 0.5158>
 }
 def_bandolier_front_01_ctrl_null {type=Null, loc=<12.3146 -2.6026 140.7081>, rot=<0.3565 0.7608 0.2301 0.4911>}
 def_bandolier_front_01_ctrl {controlType=EULER_TRANSFORM, parent=def_bandolier_front_01_ctrl_null, type=Control, loc=<0 0 0>, rot=<0 0 0 1>}
 def_bandolier_front_02_ctrl_null {type=Null, loc=<10.8784 1.8769 138.0742>, rot=<-0.2224 -0.8204 -0.1378 -0.5085>}
 def_bandolier_front_02_ctrl {
  controlType=EULER_TRANSFORM
  parent=def_bandolier_front_02_ctrl_null
  type=Control
  loc=<0.3505 5.6821 138.2981>
  rot=<0.0695 0.8408 0.2169 0.4911>
 }
 def_bandolier_front_03_ctrl_null {type=Null, loc=<9.2156 5.7521 134.397>, rot=<-0.1259 -0.8531 -0.0739 -0.501>}
 def_bandolier_front_03_ctrl {
  controlType=EULER_TRANSFORM
  parent=def_bandolier_front_03_ctrl_null
  type=Control
  loc=<-3.4927 7.0536 134.146>
  rot=<0.018 0.8515 0.1271 0.5085>
 }
 def_bandolier_front_04_ctrl_null {type=Null, loc=<7.2305 9.8262 129.523>, rot=<-0.07 -0.861 -0.0408 -0.5021>}
 def_bandolier_front_04_ctrl {
  controlType=EULER_TRANSFORM
  parent=def_bandolier_front_04_ctrl_null
  type=Control
  loc=<-0.6995 7.6352 128.8732>
  rot=<0.0131 0.8623 0.0732 0.501>
 }
 def_bandolier_front_05_ctrl_null {type=Null, loc=<4.8733 12.2774 124.7525>, rot=<-0.0441 -0.8635 -0.0256 -0.5018>}
 def_bandolier_front_05_ctrl {
  controlType=EULER_TRANSFORM
  parent=def_bandolier_front_05_ctrl_null
  type=Control
  loc=<-1.7811 10.6893 124.4471>
  rot=<0.018 0.8637 0.0406 0.5021>
 }
 def_bandolier_front_06_ctrl_null {type=Null, loc=<0.8224 11.574 117.6421>, rot=<0.0119 -0.8759 0.0066 -0.4822>}
 def_bandolier_front_06_ctrl {
  controlType=EULER_TRANSFORM
  parent=def_bandolier_front_06_ctrl_null
  type=Control
  loc=<-14.3291 11.9154 117.1183>
  rot=<-0.0687 0.8619 0.0236 0.5018>
 }
 def_bandolier_front_07_ctrl_null {type=Null, loc=<-2.8016 10.5558 111.9169>, rot=<0.0662 -0.8786 0.0355 -0.4716>}
 def_bandolier_front_07_ctrl {
  controlType=EULER_TRANSFORM
  parent=def_bandolier_front_07_ctrl_null
  type=Control
  loc=<-13.2766 9.7677 111.6944>
  rot=<-0.1193 0.8679 -0.0063 0.4822>
 }
 def_bandolier_front_08_ctrl_null {type=Null, loc=<-7.5366 9.2578 104.7765>, rot=<0.106 -0.8848 0.054 -0.4504>}
 def_bandolier_front_08_ctrl {
  controlType=EULER_TRANSFORM
  parent=def_bandolier_front_08_ctrl_null
  type=Control
  loc=<-21.368 6.7226 104.0411>
  rot=<-0.1418 0.8697 -0.0326 0.4716>
 }
 def_bandolier_front_09_ctrl_null {type=Null, loc=<-11.1559 7.7622 99.8027>, rot=<0.1683 -0.8926 0.0775 -0.411>}
 def_bandolier_front_09_ctrl {
  controlType=EULER_TRANSFORM
  parent=def_bandolier_front_09_ctrl_null
  type=Control
  loc=<-25.6355 2.5112 98.3012>
  rot=<-0.2198 0.8641 -0.0455 0.4504>
 }
 def_bandolier_front_10_ctrl_null {type=Null, loc=<-15.3135 5.8252 95.1862>, rot=<0.2691 -0.8339 -0.1813 -0.4464>}
 def_bandolier_front_10_ctrl {
  controlType=EULER_TRANSFORM
  parent=def_bandolier_front_10_ctrl_null
  type=Control
  loc=<-1.7751 -44.2951 87.3464>
  rot=<-0.114 0.8026 0.417 0.411>
 }
 def_bandolier_front_strap_ctrl_null {type=Null, loc=<-18.9783 3.5762 92.9361>, rot=<-0.3196 0.7888 0.3012 0.43>}
 def_bandolier_front_strap_ctrl {
  controlType=EULER_TRANSFORM
  parent=def_bandolier_front_strap_ctrl_null
  type=Control
  loc=<-17.5959 -22.4775 93.9146>
  rot=<0.2477 -0.7539 -0.4134 -0.4464>
 }
 dyn_bag_sim_chain_03_ctrl_null {type=Null, loc=<-18.1588 -0.1036 92.9208>, rot=<0.572 0.1174 -0.8109 0.037>}
 dyn_bag_sim_chain_03_ctrl {controlType=EULER_TRANSFORM, parent=dyn_bag_sim_chain_03_ctrl_null, type=Control, loc=<0 0 0>, rot=<0 0 0 1>}
 thigh_r_ctrl {controlType=EULER_TRANSFORM, type=Control, loc=<0 0 0>, rot=<0 0 0 1>}
]
variables [
 Crouch {type=double, default=0}
 CrouchMOV {type=double, default=0}
 Emote {type=double, default=0}
 FE {type=bool, default=False}
 Jump {type=double, default=0}
 KneeSlide {type=double, default=0}
 RBAN {type=bool, default=False}
 Spine_Rot_L {type=double, default=0}
 Spine_Rot_r {type=double, default=0}
 bandolier_front_ctrl {
  type="TArray<FRigElementKey>"
  default=((Type=Control,Name="def_bandolier_front_01_ctrl"),(Type=Control,Name="def_bandolier_front_02_ctrl"),(Type=Control,Name="def_bandolier_front_03_ctrl"),(Type=Control,Name="def_bandolier_front_04_ctrl"),(Type=Control,Name="def_bandolier_front_05_ctrl"))
 }
 bandolier_front_ctrl_NORBAN {
  type="TArray<FRigElementKey>"
  default="((Type=Control,Name=\"def_bandolier_front_01_ctrl\"),(Type=Control,Name=\"def_bandolier_front_02_ctrl\"),(Type=Control,Name=\"def_bandolier_front_03_ctrl\"),(Type=Control,Name=\"def_bandolier_front_04_ctrl\"),(Type=Control,Name=\"def_bandolier_front_05_ctrl\"),(Type ...(truncated)"
 }
 bandolier_front_joints {
  type="TArray<FRigElementKey>"
  default=((Type=Bone,Name="dyn_bandolier_front_01"),(Type=Bone,Name="dyn_bandolier_front_02"),(Type=Bone,Name="dyn_bandolier_front_03"),(Type=Bone,Name="dyn_bandolier_front_04"),(Type=Bone,Name="dyn_bandolier_front_05"))
 }
 bandolier_front_joints_NORBAN {
  type="TArray<FRigElementKey>"
  default="((Type=Bone,Name=\"dyn_bandolier_front_01\"),(Type=Bone,Name=\"dyn_bandolier_front_02\"),(Type=Bone,Name=\"dyn_bandolier_front_03\"),(Type=Bone,Name=\"dyn_bandolier_front_04\"),(Type=Bone,Name=\"dyn_bandolier_front_05\"),(Type=Bone,Name=\"dyn_bandolier_front_06\"),(Ty ...(truncated)"
 }
 bandolier_rear_ctrl {
  type="TArray<FRigElementKey>"
  default=((Type=Control,Name="def_bandolier_rear_01_ctrl"),(Type=Control,Name="def_bandolier_rear_02_ctrl"),(Type=Control,Name="def_bandolier_rear_03_ctrl"),(Type=Control,Name="def_bandolier_rear_04_ctrl"))
 }
 bandolier_rear_ctrl_NORBAN {
  type="TArray<FRigElementKey>"
  default="((Type=Control,Name=\"def_bandolier_rear_01_ctrl\"),(Type=Control,Name=\"def_bandolier_rear_02_ctrl\"),(Type=Control,Name=\"def_bandolier_rear_03_ctrl\"),(Type=Control,Name=\"def_bandolier_rear_04_ctrl\"),(Type=Control,Name=\"def_bandolier_rear_05_ctrl\"),(Type=Cont ...(truncated)"
 }
 bandolier_rear_joints {
  type="TArray<FRigElementKey>"
  default=((Type=Bone,Name="dyn_bandolier_rear_01"),(Type=Bone,Name="dyn_bandolier_rear_02"),(Type=Bone,Name="dyn_bandolier_rear_03"),(Type=Bone,Name="dyn_bandolier_rear_04"))
 }
 "bandolier_rear_joints_NORBAN " {
  type="TArray<FRigElementKey>"
  default="((Type=Bone,Name=\"dyn_bandolier_rear_01\"),(Type=Bone,Name=\"dyn_bandolier_rear_02\"),(Type=Bone,Name=\"dyn_bandolier_rear_03\"),(Type=Bone,Name=\"dyn_bandolier_rear_04\"),(Type=Bone,Name=\"dyn_bandolier_rear_05\"),(Type=Bone,Name=\"dyn_bandolier_rear_06\"),(Type=Bon ...(truncated)"
 }
 grind {type=double, default=0}
 pickaxe {type=double, default=0}
 pinLowerBack {type=double, default=0}
]
