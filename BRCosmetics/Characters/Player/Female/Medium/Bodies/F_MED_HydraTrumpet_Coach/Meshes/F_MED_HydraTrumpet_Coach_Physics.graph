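# Physics asset for the F_MED_HydraTrumpet_Coach female cosmetic body, modeling a bandolier and bag with capsule bodies alongside root, pelvis, spine and right-thigh capsules.
# Front and rear strap chains run from base_bandolier on spine_03 to two bag bodies, mostly with tight angular limits and locked twist near the bag.
# The chains form a closed loop through bag-to-bag and cross-strap constraints, with a small helper chain tying both straps back to spine_03.
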PhysicsAsset /BRCosmetics/Characters/Player/Female/Medium/Bodies/F_MED_HydraTrumpet_Coach/Meshes/F_MED_HydraTrumpet_Coach_Physics
bodies {
 base_bandolier [{shapeType=Capsule, length=0.505, radius=0.505, shapeName=base_bandolier_capsule, rotation=<rot 0 -90 0>}]
 dyn_bag_front_01 [{
   shapeType=Capsule
   length=15.6919
   radius=8.5
   shapeName=dyn_bag_front_01_capsule
   center=<1.943 2.3324 8.5512>
   rotation=<rot 74.1597 -18.0811 63.1754>
  }]
 dyn_bag_rear_01 [
  {
   shapeType=Capsule
   length=15.9378
   radius=8
   shapeName=dyn_bag_rear_01_capsule
   center=<14.3807 -2.3291 2.875>
   rotation=<rot 81.1966 -22.8748 57.3527>
  }
  {
   shapeType=Capsule
   length=17
   radius=8.5
   shapeName=dyn_bag_rear_01_capsule
   center=<9.0321 -3.0462 5.7057>
   rotation=<rot 74.6367 0 -11.2102>
  }
 ]
 dyn_bag_sim_chain_02 [{shapeType=Capsule, length=1, radius=1, shapeName=dyn_bag_sim_chain_02_capsule, rotation=<rot 90 0 0>}]
 dyn_bag_sim_chain_03 [{shapeType=Capsule, length=1, radius=1, shapeName=dyn_bag_sim_chain_03_capsule, center=<1 0 0>, rotation=<rot 90 0 0>}]
 dyn_bandolier_front_05 [{
   shapeType=Capsule
   length=10.7034
   radius=2.3319
   shapeName=dyn_bandolier_front_05_capsule
   center=<8.0322 -0.3633 -0.5848>
   rotation=<rot 0 -102.0683 -84.3041>
  }]
 dyn_bandolier_front_06 [{shapeType=Capsule, length=2, radius=2, shapeName=dyn_bandolier_front_06_capsule, center=<2 0 0>, rotation=<rot 90 0 0>}]
 dyn_bandolier_front_07 [{shapeType=Capsule, length=2, radius=2, shapeName=dyn_bandolier_front_07_capsule, center=<2 0 0>, rotation=<rot 90 0 0>}]
 dyn_bandolier_front_08 [{shapeType=Capsule, length=2, radius=2, shapeName=dyn_bandolier_front_08_capsule, center=<2 0 0>, rotation=<rot 90 0 0>}]
 dyn_bandolier_front_09 [{shapeType=Capsule, length=3.5, radius=2, shapeName=dyn_bandolier_front_09_capsule, center=<2 0 0>, rotation=<rot 90 0 0>}]
 dyn_bandolier_front_10 [{shapeType=Capsule, length=4.5, radius=2, shapeName=dyn_bandolier_front_10_capsule, center=<4.4023 0 0>, rotation=<rot 90 0 0>}]
 dyn_bandolier_front_strap [{shapeType=Capsule, length=3.5, radius=2, shapeName=dyn_bandolier_front_strap_capsule, center=<2 0.7386 0>, rotation=<rot 90 0 0>}]
 dyn_bandolier_rear_06 [{shapeType=Capsule, length=2, radius=2, shapeName=dyn_bandolier_rear_06_capsule, center=<2 0 0>, rotation=<rot 90 0 0>}]
 dyn_bandolier_rear_07 [{shapeType=Capsule, length=2, radius=2, shapeName=dyn_bandolier_rear_07_capsule, center=<2 0 0>, rotation=<rot 90 0 0>}]
 dyn_bandolier_rear_08 [{shapeType=Capsule, length=2, radius=2, shapeName=dyn_bandolier_rear_08_capsule, center=<2 0 0>, rotation=<rot 90 0 0>}]
 dyn_bandolier_rear_09 [{shapeType=Capsule, length=2, radius=2, shapeName=dyn_bandolier_rear_09_capsule, center=<2 0 0>, rotation=<rot 90 0 0>}]
 dyn_bandolier_rear_10 [{shapeType=Capsule, length=3.5, radius=2, shapeName=dyn_bandolier_rear_10_capsule, center=<2.9466 0 0>, rotation=<rot 90 0 0>}]
 dyn_bandolier_rear_strap [{shapeType=Capsule, length=2, radius=2, shapeName=dyn_bandolier_rear_strap_capsule, center=<2 0 0>, rotation=<rot 90 0 0>}]
 pelvis [
  {shapeType=Capsule, length=12.0876, radius=9.9997, shapeName=pelvis_capsule, center=<2 -1.1313 -0.3136>, rotation=<rot -1.0649 -90 -1.3918>}
  {shapeType=Capsule, length=12.5, radius=9.5, shapeName=pelvis_capsule, center=<-4.4748 -3.7088 0>}
 ]
 root [{shapeType=Capsule, length=0.505, radius=0.505, shapeName=root_capsule, center=<2 0 0>, rotation=<rot 0 -90 0>}]
 spine_01 [{shapeType=Capsule, length=7.7192, radius=10.2808, shapeName=spine_01_capsule, center=<2 0 0>, rotation=<rot 0 -90 0>}]
 spine_02 [{shapeType=Capsule, length=4.5799, radius=11.4201, shapeName=spine_02_capsule, center=<2 0 0>}]
 spine_03 [{shapeType=Capsule, length=8.7599, radius=14.1727, shapeName=spine_03_capsule, center=<6.1339 0 0>}]
 thigh_r [{
   shapeType=Capsule
   length=26.0291
   radius=8.8403
   shapeName=thigh_r_capsule
   center=<20.8144 -1.5718 -0.4696>
   rotation=<rot 0.1195 -87.9985 -87.2162>
  }]
}
constraints [
 {
  bone1Name=dyn_bag_sim_chain_03
  bone2Name=dyn_bag_sim_chain_02
  swing1LimitDegrees=2
  swing1Motion=Free
  swing2LimitDegrees=2
  swing2Motion=Free
  twistLimitDegrees=2
  twistMotion=Free
 }
 {
  bone1Name=pelvis
  bone2Name=root
  swing1LimitDegrees=45
  swing1Motion=Free
  swing2LimitDegrees=45
  swing2Motion=Free
  twistLimitDegrees=45
  twistMotion=Free
 }
 {
  bone1Name=spine_01
  bone2Name=pelvis
  swing1LimitDegrees=45
  swing1Motion=Free
  swing2LimitDegrees=45
  swing2Motion=Free
  twistLimitDegrees=45
  twistMotion=Free
 }
 {
  bone1Name=spine_02
  bone2Name=spine_01
  swing1LimitDegrees=45
  swing1Motion=Free
  swing2LimitDegrees=45
  swing2Motion=Free
  twistLimitDegrees=45
  twistMotion=Free
 }
 {
  bone1Name=spine_03
  bone2Name=spine_02
  swing1LimitDegrees=45
  swing1Motion=Free
  swing2LimitDegrees=45
  swing2Motion=Free
  twistLimitDegrees=45
  twistMotion=Free
 }
 {
  bone1Name=thigh_r
  bone2Name=pelvis
  swing1LimitDegrees=45
  swing1Motion=Free
  swing2LimitDegrees=45
  swing2Motion=Free
  twistLimitDegrees=45
  twistMotion=Free
 }
 {
  bone1Name=dyn_bag_sim_chain_02
  bone2Name=spine_03
  swing1LimitDegrees=45
  swing1Motion=Locked
  swing2LimitDegrees=30
  twistLimitDegrees=45
  twistMotion=Locked
 }
 {
  bone1Name=base_bandolier
  bone2Name=spine_03
  swing1LimitDegrees=45
  swing1Motion=Free
  swing2LimitDegrees=45
  swing2Motion=Free
  twistLimitDegrees=45
  twistMotion=Free
 }
 {bone1Name=dyn_bandolier_front_06, bone2Name=dyn_bandolier_front_05, swing1LimitDegrees=2, swing2LimitDegrees=2, twistLimitDegrees=2}
 {bone1Name=dyn_bandolier_front_07, bone2Name=dyn_bandolier_front_06, swing1LimitDegrees=2, swing2LimitDegrees=20, twistLimitDegrees=2}
 {bone1Name=dyn_bandolier_front_08, bone2Name=dyn_bandolier_front_07, swing1LimitDegrees=2, swing2LimitDegrees=20, twistLimitDegrees=2}
 {
  bone1Name=dyn_bandolier_front_09
  bone2Name=dyn_bandolier_front_08
  swing1LimitDegrees=5
  swing2LimitDegrees=5
  twistLimitDegrees=2
  twistMotion=Locked
 }
 {
  bone1Name=dyn_bandolier_front_10
  bone2Name=dyn_bandolier_front_09
  swing1LimitDegrees=2
  swing2LimitDegrees=2
  twistLimitDegrees=2
  twistMotion=Locked
 }
 {
  bone1Name=dyn_bandolier_front_strap
  bone2Name=dyn_bandolier_front_10
  swing1LimitDegrees=20
  swing2LimitDegrees=20
  twistLimitDegrees=0
  twistMotion=Locked
 }
 {
  bone1Name=dyn_bag_front_01
  bone2Name=dyn_bandolier_front_strap
  swing1LimitDegrees=2
  swing2LimitDegrees=20
  twistLimitDegrees=0
  twistMotion=Locked
 }
 {bone1Name=dyn_bandolier_rear_06, bone2Name=base_bandolier, swing1LimitDegrees=5, swing2LimitDegrees=20, twistLimitDegrees=0}
 {bone1Name=dyn_bandolier_rear_07, bone2Name=dyn_bandolier_rear_06, swing1LimitDegrees=2, swing2LimitDegrees=10, twistLimitDegrees=2}
 {bone1Name=dyn_bandolier_rear_08, bone2Name=dyn_bandolier_rear_07, swing1LimitDegrees=2, swing2LimitDegrees=10, twistLimitDegrees=2}
 {bone1Name=dyn_bandolier_rear_09, bone2Name=dyn_bandolier_rear_08, swing1LimitDegrees=2, swing2LimitDegrees=2, twistLimitDegrees=2}
 {
  bone1Name=dyn_bandolier_rear_10
  bone2Name=dyn_bandolier_rear_09
  swing1LimitDegrees=2
  swing2LimitDegrees=2
  twistLimitDegrees=2
  twistMotion=Locked
 }
 {
  bone1Name=dyn_bandolier_rear_strap
  bone2Name=dyn_bandolier_rear_10
  swing1LimitDegrees=20
  swing2LimitDegrees=20
  twistLimitDegrees=0
  twistMotion=Locked
 }
 {
  bone1Name=dyn_bag_rear_01
  bone2Name=dyn_bandolier_rear_strap
  swing1LimitDegrees=20
  swing1Motion=Locked
  swing2LimitDegrees=20
  twistLimitDegrees=0
  twistMotion=Locked
 }
 {
  bone1Name=dyn_bag_front_01
  bone2Name=dyn_bag_rear_01
  swing1LimitDegrees=20
  swing1Motion=Locked
  swing2LimitDegrees=4
  twistLimitDegrees=20
  twistMotion=Locked
 }
 {
  bone1Name=dyn_bandolier_front_05
  bone2Name=base_bandolier
  swing1LimitDegrees=45
  swing1Motion=Free
  swing2LimitDegrees=45
  swing2Motion=Free
  twistLimitDegrees=45
  twistMotion=Free
 }
 {
  bone1Name=dyn_bandolier_front_strap
  bone2Name=dyn_bag_sim_chain_03
  swing1LimitDegrees=45
  swing1Motion=Free
  swing2LimitDegrees=45
  swing2Motion=Free
  twistLimitDegrees=45
  twistMotion=Free
 }
 {
  bone1Name=dyn_bandolier_rear_strap
  bone2Name=dyn_bag_sim_chain_03
  swing1LimitDegrees=45
  swing1Motion=Free
  swing2LimitDegrees=45
  swing2Motion=Free
  twistLimitDegrees=45
  twistMotion=Free
 }
 {
  bone1Name=dyn_bandolier_front_10
  bone2Name=dyn_bandolier_rear_strap
  swing1LimitDegrees=45
  swing1Motion=Free
  swing2LimitDegrees=45
  swing2Motion=Free
  twistLimitDegrees=45
  twistMotion=Free
 }
]
