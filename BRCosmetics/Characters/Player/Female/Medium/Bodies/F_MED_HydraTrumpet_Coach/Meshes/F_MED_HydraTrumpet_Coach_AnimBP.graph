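# Post-process-style AnimBP for the F_MED_HydraTrumpet_Coach player body, adding secondary motion to laces, keychains, bags and bandana.
# Copied parent pose runs through RigidBody and chained AnimDynamics, then bone offsets for backbling and baseball slides.
# EventGraph computes movement-state floats, state-based keychain gravity, and WPO cloth material wind and foot/calf inputs.
# The Control Rig, Constraint and some bandolier/bag dynamics nodes are disconnected from the output.
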AnimBlueprint /BRCosmetics/Characters/Player/Female/Medium/Bodies/F_MED_HydraTrumpet_Coach/Meshes/F_MED_HydraTrumpet_Coach_AnimBP
target_skeleton=&/F_MED_HydraTrumpet_Coach_Skeleton
graphs [
 AnimGraph {nodes [
   n1 = Anim Dynamics - Bone: dyn_bandolier_rear_07(Alpha=1, ExternalForce="0,0,0") {
    angularDampingOverride=1
    angularSpringConstant=5000
    bAngularSpring=true
    bOverrideAngularDamping=true
    bUseGravityOverride=true
    gravityScale=0.35
    simulationSpace=BoneRelative
    alphaBoolBlend {blendOption=Linear}
    boundBone {boneName=dyn_bandolier_rear_07}
    componentAppliedLinearAccClamp=<80000 80000 80000>
    componentLinearAccScale=<0.2 0.2 0.2>
    componentLinearVelScale=<0.2 0.2 0.2>
    gravityOverride {z=-980}
    physicsBodyDefinitions [{
      boundBone {boneName=dyn_bandolier_rear_07}
      boxExtents=<10 5 5>
      constraintSetup {angularLimitsMax=<0 10 0>, angularLimitsMin=<0 0 0>, angularTarget=<1 0 0>, linearAxesMin=<0 -1 -1>}
      localJointOffset=<2 0 0>
     }]
    relativeSpaceBone {boneName=base_bandolier_rear_07}
   }
   Anim Dynamics - Bone: dyn_bandolier_rear_08(ComponentPose=@n1, Alpha=1, ExternalForce="0,0,0") {
    angularDampingOverride=1
    angularSpringConstant=5000
    bAngularSpring=true
    bOverrideAngularDamping=true
    bUseGravityOverride=true
    gravityScale=0.35
    simulationSpace=BoneRelative
    alphaBoolBlend {blendOption=Linear}
    boundBone {boneName=dyn_bandolier_rear_08}
    componentAppliedLinearAccClamp=<80000 80000 80000>
    componentLinearAccScale=<0.2 0.2 0.2>
    componentLinearVelScale=<0.2 0.2 0.2>
    gravityOverride {z=-980}
    physicsBodyDefinitions [{
      boundBone {boneName=dyn_bandolier_rear_08}
      boxExtents=<10 5 5>
      constraintSetup {angularLimitsMax=<0 20 0>, angularLimitsMin=<0 0 0>, angularTarget=<1 0 0>, linearAxesMin=<0 -1 -1>}
      localJointOffset=<2 0 0>
     }]
    relativeSpaceBone {boneName=base_bandolier_rear_08}
   }
   n2 = Anim Dynamics - Bone: dyn_bag_rear_02(Alpha=1, ExternalForce="0,0,0") {
    angularDampingOverride=1
    angularSpringConstant=20000
    bAngularSpring=true
    bOverrideAngularDamping=true
    bUseGravityOverride=true
    gravityScale=0.35
    simulationSpace=RootRelative
    alphaBoolBlend {blendOption=Linear}
    boundBone {boneName=dyn_bag_rear_02}
    componentAppliedLinearAccClamp=<80000 80000 80000>
    componentLinearAccScale=<0.5 0.5 0.5>
    componentLinearVelScale=<0.5 0.5 0.5>
    gravityOverride {z=-980}
    physicsBodyDefinitions [{
      boundBone {boneName=dyn_bag_rear_02}
      boxExtents=<20 30 30>
      constraintSetup {angularLimitsMax=<10 10 10>, angularLimitsMin=<-10 -10 -10>, angularTarget=<1 0 0>}
      localJointOffset=<4 0 0>
     }]
   }
   Anim Dynamics - Bone: dyn_bag_front_01(ComponentPose=@n2, Alpha=1, ExternalForce="0,0,0") {
    angularDampingOverride=1
    angularSpringConstant=20000
    bAngularSpring=true
    bOverrideAngularDamping=true
    bUseGravityOverride=true
    gravityScale=0.35
    simulationSpace=RootRelative
    alphaBoolBlend {blendOption=Linear}
    boundBone {boneName=dyn_bag_front_01}
    componentAppliedLinearAccClamp=<80000 80000 80000>
    componentLinearAccScale=<0.5 0.5 0.5>
    componentLinearVelScale=<0.5 0.5 0.5>
    gravityOverride {z=-980}
    physicsBodyDefinitions [{
      boundBone {boneName=dyn_bag_front_01}
      boxExtents=<20 30 30>
      constraintSetup {angularLimitsMax=<10 10 10>, angularLimitsMin=<-10 -10 -10>, angularTarget=<1 0 0>}
      localJointOffset=<4 0 0>
     }]
   }
   "Constraint - base_bag_rear_01 (P)"(Alpha=1, ConstraintWeights_0=1) {
    alphaBoolBlend {blendOption=Linear}
    boneToModify {boneName=base_bag_rear_01}
    constraintSetup [{targetBone {boneName=base_bag_front}}]
    constraintWeights [1]
   }
   Control Rig!(
    Source
    Alpha
    Crouch:double=Get Crouching() -> (Crouching:double)
    CrouchMOV:double=Get CrouchMOV() -> (CrouchMOV:double)
    Spine_Rot_r:double=Get spine_rot_r() -> (spine_rot_r:double)
    Spine_Rot_L:double=Get spine_rot_L() -> (spine_rot_L:double)
    Jump:double=Get Isjumpo() -> (Isjumpo:double)
    pinLowerBack:double=Get PinLower Back() -> (PinLower Back:double)
    RBAN:bool=Get bRigidBodyEnabled()
    pickaxe:double=Get pickaxe() -> (pickaxe:double)
    Emote:double=Get Emote() -> (Emote:double)
    grind:double=Get RailGrind() -> (RailGrind:double)
    KneeSlide:double=Get Knees() -> (Knees:double)
   ) -> (Pose) {
    alphaBoolBlend {blendOption=Linear}
    controlRigAssetReference {blueprintRigClass=&/F_MED_HydraTrumpet_Coach_CtrlRig.F_MED_HydraTrumpet_Coach_CtrlRig_C}
   }
   n3 = RigidBody!(
    ComponentPose=Local To Component(LocalPose=Copy Pose From Mesh() {bUseAttachedParent=true})
    Alpha
    OverrideWorldGravity:Vector="0,0,-980"
    ExternalForce
   ) -> (Pose) {
    bOverrideWorldGravity=true
    overridePhysicsAsset=&/F_MED_HydraTrumpet_Coach_Physics
    simulationSpace=BaseBoneSpace
    alphaBoolBlend {blendOption=Linear}
    alphaScaleBiasClamp {bInterpResult=true}
    baseBoneRef {boneName=pelvis}
    componentAppliedLinearAccClamp=<0 0 0>
    overrideWorldGravity {z=-980}
    simSpaceSettings {
     dampingAlpha=0
     maxAngularAcceleration=5
     maxAngularVelocity=20
     maxLinearAcceleration=3000
     maxLinearVelocity=600
     velocityScaleZ=0.5
     worldAlpha=1
     externalLinearDragV=<1 1 0.9>
    }
   }
   n4 = Anim Dynamics - Bone: dyn_lace_r(ComponentPose=@n3, Alpha=1, ExternalForce="0,0,0") {
    angularDampingOverride=1
    angularSpringConstant=200
    bAngularSpring=true
    bOverrideAngularDamping=true
    bUseGravityOverride=true
    gravityScale=0.35
    simulationSpace=RootRelative
    alphaBoolBlend {blendOption=Linear}
    boundBone {boneName=dyn_lace_r}
    componentAppliedLinearAccClamp=<3000 3000 3000>
    componentLinearAccScale=<1 1 1>
    componentLinearVelScale=<1 1 1>
    gravityOverride {z=-980}
    physicsBodyDefinitions [{
      boundBone {boneName=dyn_lace_r}
      boxExtents=<10 10 10>
      constraintSetup {angularLimitsMax=<15 15 15>, angularLimitsMin=<-15 -15 -30>, angularTarget=<1 0 0>}
      localJointOffset=<5 0 0>
     }]
   }
   n5 = Anim Dynamics - Bone: dyn_lace_l(ComponentPose=@n4, Alpha=1, ExternalForce="0,0,0") {
    angularDampingOverride=1
    angularSpringConstant=200
    bAngularSpring=true
    bOverrideAngularDamping=true
    bUseGravityOverride=true
    gravityScale=0.35
    simulationSpace=RootRelative
    alphaBoolBlend {blendOption=Linear}
    boundBone {boneName=dyn_lace_l}
    componentAppliedLinearAccClamp=<3000 3000 3000>
    componentLinearAccScale=<1 1 1>
    componentLinearVelScale=<1 1 1>
    gravityOverride {z=-980}
    physicsBodyDefinitions [{
      boundBone {boneName=dyn_lace_l}
      boxExtents=<10 10 10>
      constraintSetup {angularLimitsMax=<15 15 15>, angularLimitsMin=<-15 -15 -30>, angularTarget=<1 0 0>}
      localJointOffset=<5 0 0>
     }]
   }
   n6 = Anim Dynamics - Bone: dyn_lace(ComponentPose=@n5, Alpha=1, ExternalForce="0,0,0") {
    angularDampingOverride=1
    angularSpringConstant=200
    bAngularSpring=true
    bOverrideAngularDamping=true
    bUseGravityOverride=true
    gravityScale=0.35
    simulationSpace=RootRelative
    alphaBoolBlend {blendOption=Linear}
    boundBone {boneName=dyn_lace}
    componentAppliedLinearAccClamp=<3000 3000 3000>
    componentLinearAccScale=<1 1 1>
    componentLinearVelScale=<1 1 1>
    gravityOverride {z=-980}
    physicsBodyDefinitions [{
      boundBone {boneName=dyn_lace}
      boxExtents=<10 10 10>
      constraintSetup {angularLimitsMax=<15 15 15>, angularLimitsMin=<-15 -15 -60>, angularTarget=<1 0 0>}
      localJointOffset=<5 0 0>
     }]
   }
   n7 = Get Gravity Override()
   n8 = Anim Dynamics - Bone: dyn_keychain!(ComponentPose=@n6, Alpha=1, GravityOverride:Vector=@n7, ExternalForce="0,0,0") -> (Pose) {
    angularDampingOverride=0.99
    angularSpringConstant=1000
    bAngularSpring=true
    bOverrideAngularDamping=true
    bUseGravityOverride=true
    gravityScale=0.35
    simulationSpace=RootRelative
    alphaBoolBlend {blendOption=Linear}
    boundBone {boneName=dyn_keychain}
    componentAppliedLinearAccClamp=<3000 3000 3000>
    componentLinearAccScale=<0.6 0.6 0.6>
    componentLinearVelScale=<0.6 0.6 0.6>
    gravityOverride {z=-980}
    physicsBodyDefinitions [{
      boundBone {boneName=dyn_keychain}
      boxExtents=<20 30 30>
      constraintSetup {angularLimitsMax=<0 0 10>, angularLimitsMin=<-20 -20 -20>, angularTarget=<1 0 0>}
      localJointOffset=<4 0 0>
     }]
   }
   n9 = Anim Dynamics - Bone: dyn_keychain1!(ComponentPose=@n8, Alpha=1, GravityOverride:Vector=@n7, ExternalForce="0,0,0") -> (Pose) {
    angularDampingOverride=0.99
    angularSpringConstant=1000
    bAngularSpring=true
    bOverrideAngularDamping=true
    bUseGravityOverride=true
    gravityScale=0.35
    simulationSpace=RootRelative
    alphaBoolBlend {blendOption=Linear}
    boundBone {boneName=dyn_keychain1}
    componentAppliedLinearAccClamp=<3000 3000 3000>
    componentLinearAccScale=<0.6 0.6 0.6>
    componentLinearVelScale=<0.6 0.6 0.6>
    gravityOverride {z=-980}
    physicsBodyDefinitions [{
      boundBone {boneName=dyn_keychain1}
      boxExtents=<20 30 30>
      constraintSetup {angularLimitsMax=<0 0 10>, angularLimitsMin=<-20 -20 -20>, angularTarget=<1 0 0>}
      localJointOffset=<4 0 0>
     }]
   }
   n10 = Anim Dynamics - Bone: dyn_keychain2!(ComponentPose=@n9, Alpha=1, GravityOverride:Vector=@n7, ExternalForce="0,0,0") -> (Pose) {
    angularDampingOverride=0.99
    angularSpringConstant=1000
    bAngularSpring=true
    bOverrideAngularDamping=true
    bUseGravityOverride=true
    gravityScale=0.35
    simulationSpace=RootRelative
    alphaBoolBlend {blendOption=Linear}
    boundBone {boneName=dyn_keychain2}
    componentAppliedLinearAccClamp=<3000 3000 3000>
    componentLinearAccScale=<0.6 0.6 0.6>
    componentLinearVelScale=<0.6 0.6 0.6>
    gravityOverride {z=-980}
    physicsBodyDefinitions [{
      boundBone {boneName=dyn_keychain2}
      boxExtents=<20 30 30>
      constraintSetup {angularLimitsMax=<0 0 10>, angularLimitsMin=<-20 -20 -20>, angularTarget=<1 0 0>}
      localJointOffset=<4 0 0>
     }]
   }
   n11 = Anim Dynamics - Bone: dyn_keychain3!(ComponentPose=@n10, Alpha=1, GravityOverride:Vector=@n7, ExternalForce="0,0,0") -> (Pose) {
    angularDampingOverride=0.99
    angularSpringConstant=1000
    bAngularSpring=true
    bOverrideAngularDamping=true
    bUseGravityOverride=true
    gravityScale=0.35
    simulationSpace=RootRelative
    alphaBoolBlend {blendOption=Linear}
    boundBone {boneName=dyn_keychain3}
    componentAppliedLinearAccClamp=<3000 3000 3000>
    componentLinearAccScale=<0.6 0.6 0.6>
    componentLinearVelScale=<0.6 0.6 0.6>
    gravityOverride {z=-980}
    physicsBodyDefinitions [{
      boundBone {boneName=dyn_keychain3}
      boxExtents=<20 30 30>
      constraintSetup {angularLimitsMax=<0 0 10>, angularLimitsMin=<-20 -20 -40>, angularTarget=<1 0 0>}
      localJointOffset=<6 0 0>
     }]
   }
   n12 = Anim Dynamics - Bone: dyn_bag_rear_03(ComponentPose=@n11, Alpha=1, ExternalForce="0,0,0") {
    angularDampingOverride=1
    angularSpringConstant=20000
    bAngularSpring=true
    bOverrideAngularDamping=true
    bUseGravityOverride=true
    gravityScale=0.35
    simulationSpace=RootRelative
    alphaBoolBlend {blendOption=Linear}
    boundBone {boneName=dyn_bag_rear_03}
    componentAppliedLinearAccClamp=<3000 3000 3000>
    componentLinearAccScale=<0.5 0.5 0.5>
    componentLinearVelScale=<0.5 1 0.5>
    gravityOverride {z=-980}
    physicsBodyDefinitions [{
      boundBone {boneName=dyn_bag_rear_03}
      boxExtents=<20 30 30>
      constraintSetup {angularLimitsMax=<15 15 15>, angularLimitsMin=<-15 -5 -15>, angularTarget=<1 0 0>}
      localJointOffset=<4 0 0>
     }]
   }
   n13 = Anim Dynamics - Bone: dyn_bag_rear_04(ComponentPose=@n12, Alpha=1, ExternalForce="0,0,0") {
    angularDampingOverride=1
    angularSpringConstant=20000
    bAngularSpring=true
    bOverrideAngularDamping=true
    bUseGravityOverride=true
    gravityScale=0.35
    simulationSpace=RootRelative
    alphaBoolBlend {blendOption=Linear}
    boundBone {boneName=dyn_bag_rear_04}
    componentAppliedLinearAccClamp=<3000 3000 3000>
    componentLinearAccScale=<0.5 0.5 0.5>
    componentLinearVelScale=<0.5 1 0.5>
    gravityOverride {z=-980}
    physicsBodyDefinitions [{
      boundBone {boneName=dyn_bag_rear_04}
      boxExtents=<20 30 30>
      constraintSetup {angularLimitsMax=<15 15 15>, angularLimitsMin=<-15 -15 -15>, angularTarget=<1 0 0>}
      localJointOffset=<4 0 0>
     }]
   }
   n14 = Anim Dynamics - Bone: dyn_bag_rear_05(ComponentPose=@n13, Alpha=1, ExternalForce="0,0,0") {
    angularDampingOverride=1
    angularSpringConstant=20000
    bAngularSpring=true
    bOverrideAngularDamping=true
    bUseGravityOverride=true
    gravityScale=0.35
    simulationSpace=RootRelative
    alphaBoolBlend {blendOption=Linear}
    boundBone {boneName=dyn_bag_rear_05}
    componentAppliedLinearAccClamp=<3000 3000 3000>
    componentLinearAccScale=<0.5 0.5 0.5>
    componentLinearVelScale=<0.5 1 0.5>
    gravityOverride {z=-980}
    physicsBodyDefinitions [{
      boundBone {boneName=dyn_bag_rear_05}
      boxExtents=<20 30 30>
      constraintSetup {angularLimitsMax=<15 15 15>, angularLimitsMin=<-15 -15 -15>, angularTarget=<1 0 0>}
      localJointOffset=<4 0 0>
     }]
   }
   n15 = Anim Dynamics - Bone: dyn_bag_front_02(ComponentPose=@n14, Alpha=1, ExternalForce="0,0,0") {
    angularDampingOverride=1
    angularSpringConstant=20000
    bAngularSpring=true
    bOverrideAngularDamping=true
    bUseGravityOverride=true
    gravityScale=0.35
    simulationSpace=RootRelative
    alphaBoolBlend {blendOption=Linear}
    boundBone {boneName=dyn_bag_front_02}
    componentAppliedLinearAccClamp=<3000 3000 3000>
    componentLinearAccScale=<0.5 0.5 0.5>
    componentLinearVelScale=<0.5 1 0.5>
    gravityOverride {z=-980}
    physicsBodyDefinitions [{
      boundBone {boneName=dyn_bag_front_02}
      boxExtents=<20 30 30>
      constraintSetup {angularLimitsMax=<15 15 15>, angularLimitsMin=<-15 -5 -15>, angularTarget=<1 0 0>}
      localJointOffset=<4 0 0>
     }]
   }
   n16 = Anim Dynamics - Bone: dyn_bag_front_03(ComponentPose=@n15, Alpha=1, ExternalForce="0,0,0") {
    angularDampingOverride=1
    angularSpringConstant=20000
    bAngularSpring=true
    bOverrideAngularDamping=true
    bUseGravityOverride=true
    gravityScale=0.35
    simulationSpace=RootRelative
    alphaBoolBlend {blendOption=Linear}
    boundBone {boneName=dyn_bag_front_03}
    componentAppliedLinearAccClamp=<3000 3000 3000>
    componentLinearAccScale=<0.5 0.5 0.5>
    componentLinearVelScale=<0.5 1 0.5>
    gravityOverride {z=-980}
    physicsBodyDefinitions [{
      boundBone {boneName=dyn_bag_front_03}
      boxExtents=<20 30 30>
      constraintSetup {angularLimitsMax=<15 15 15>, angularLimitsMin=<-15 -15 -15>, angularTarget=<1 0 0>}
      localJointOffset=<4 0 0>
     }]
   }
   n17 = Anim Dynamics - Bone: dyn_bag_front_04(ComponentPose=@n16, Alpha=1, ExternalForce="0,0,0") {
    angularDampingOverride=1
    angularSpringConstant=20000
    bAngularSpring=true
    bOverrideAngularDamping=true
    bUseGravityOverride=true
    gravityScale=0.35
    simulationSpace=RootRelative
    alphaBoolBlend {blendOption=Linear}
    boundBone {boneName=dyn_bag_front_04}
    componentAppliedLinearAccClamp=<3000 3000 3000>
    componentLinearAccScale=<0.5 0.5 0.5>
    componentLinearVelScale=<0.5 1 0.5>
    gravityOverride {z=-980}
    physicsBodyDefinitions [{
      boundBone {boneName=dyn_bag_front_04}
      boxExtents=<20 30 30>
      constraintSetup {angularLimitsMax=<15 15 15>, angularLimitsMin=<-15 -15 -15>, angularTarget=<1 0 0>}
      localJointOffset=<4 0 0>
     }]
   }
   n18 = Anim Dynamics - Bone: dyn_bandana(ComponentPose=@n17, Alpha=1, ExternalForce="0,0,0") {
    angularDampingOverride=1
    angularSpringConstant=10000
    bAngularSpring=true
    bOverrideAngularDamping=true
    bUseGravityOverride=true
    gravityScale=0.35
    simulationSpace=RootRelative
    alphaBoolBlend {blendOption=Linear}
    boundBone {boneName=dyn_bandana}
    componentAppliedLinearAccClamp=<3000 3000 3000>
    componentLinearAccScale=<1 1 1>
    componentLinearVelScale=<1 1 1>
    gravityOverride {z=-980}
    physicsBodyDefinitions [{
      boundBone {boneName=dyn_bandana}
      boxExtents=<20 30 30>
      constraintSetup {angularLimitsMax=<40 40 40>, angularLimitsMin=<-40 -40 -40>, angularTarget=<1 0 0>}
      localJointOffset=<4 0 0>
     }]
   }
   n19 = Get isWearingBackbling()
   n20 = "Transform (Modify) Bone - Bone: dyn_bandolier_rear_05"!(ComponentPose=@n18, bAlphaBoolEnabled:bool=@n19, Translation="0,4,0", Rotation="0,0,0", Scale="1,1,1") -> (Pose) {
    alphaInputType=Bool
    rotationMode=BMM_Replace
    rotationSpace=BCS_BoneSpace
    translationMode=BMM_Additive
    translationSpace=BCS_BoneSpace
    alphaBoolBlend {blendOption=Linear}
    boneToModify {boneName=dyn_bandolier_rear_05}
    translation {y=4}
   }
   n21 = "Transform (Modify) Bone - Bone: dyn_bandolier_rear_06"!(ComponentPose=@n20, bAlphaBoolEnabled:bool=@n19, Translation="0,-3,0", Rotation="0,0,0", Scale="1,1,1") -> (Pose) {
    alphaInputType=Bool
    rotationMode=BMM_Replace
    rotationSpace=BCS_BoneSpace
    translationMode=BMM_Additive
    translationSpace=BCS_BoneSpace
    alphaBoolBlend {blendOption=Linear}
    boneToModify {boneName=dyn_bandolier_rear_06}
    translation {y=-3}
   }
   n22 = Get PinLower Back()
   n23 = "Transform (Modify) Bone - Bone: dyn_bag_front_01"(ComponentPose=@n21, Alpha=@n22, Translation="0,0,10", Rotation="70,0,0", Scale="1,1,1") {
    rotationMode=BMM_Replace
    rotationSpace=BCS_BoneSpace
    translationMode=BMM_Additive
    translationSpace=BCS_BoneSpace
    alphaBoolBlend {blendOption=Linear}
    alphaScaleBiasClamp {bInterpResult=true, interpSpeedDecreasing=5, interpSpeedIncreasing=20}
    boneToModify {boneName=dyn_bag_front_01}
    rotation {pitch=70}
    translation {z=10}
   }
   n24 = "Transform (Modify) Bone - Bone: base_bag_rear_01"(ComponentPose=@n23, Alpha=@n22, Translation="0,0,8", Rotation="70,0,0", Scale="1,1,1") {
    rotationMode=BMM_Replace
    rotationSpace=BCS_BoneSpace
    translationMode=BMM_Additive
    translationSpace=BCS_BoneSpace
    alphaBoolBlend {blendOption=Linear}
    alphaScaleBiasClamp {bInterpResult=true, interpSpeedDecreasing=5, interpSpeedIncreasing=20}
    boneToModify {boneName=base_bag_rear_01}
    rotation {pitch=70}
    translation {z=8}
   }
   Output Pose(Result=Component To Local(ComponentPose=@n24)) {layerGroup=None}
   Get notForward() -> (notForward:double)
   Get Forward() -> (Forward:double)
   Get UpperArmRot_l() -> (UpperArmRot_l:double)
   Get is moving() -> (is moving:double)
   Get skydivve() -> (skydivve:double)
   Get spine_rot_D() -> (spine_rot_D:double)
  ]}
 EventGraph {nodes [
   n1 = Event BlueprintUpdateAnimation()
   > n2 = Sequence!(execute) -> (then_0, then_1, then_2:Exec, then_3:Exec, then_4:Exec, then_5:Exec, then_6:Exec)
   > Is Valid(InputObject=Get PlayerPawn() -> (PlayerPawn:Fort Player Pawn Athena Object Reference))
   > Set WPO Globals Function(
    Material=Get Mat Instance Dynamic()
    Overlay Material=Get Overlay Mat Instance Dynamic()
    Skeletal Mesh=Get SkelMesh()
    "Local Noise Transform  Location"=Get Local Noise Transform Location() -> (Local Noise Transform Location:Vector)
    Local Noise Offset=Get Local Noise Offset()
   )
   > Set WPO Regional Function!(
    execute
    Material:Material Instance Dynamic Object Reference=Get Mat Instance Dynamic()
    Overlay Material:Material Instance Dynamic Object Reference=Get Overlay Mat Instance Dynamic()
    Skeletal Mesh:Skeletal Mesh Component Object Reference=Get SkelMesh()
    Velocity Transform:Name=foot_r
    Velocity Interp Speed:double=Get Velocity Interp Speed()
    Collision Transform:Name=calf_r
    Collision Transform Axis:EAxis Enum
    Invert Collision Transform Axis:bool=false
    Delta Time X:double=@n1.DeltaTimeX
    Previous Velocity Transform Location:"Vector (by ref)"=Get R1 Velocity Transform Location()
    Previous Velocity Transform Velocity:"Vector (by ref)"=Get R1 Velocity Transform Velocity()
    Region:int=1
   ) -> (then)
   > Set WPO Regional Function(
    Material=Get Mat Instance Dynamic()
    Overlay Material=Get Overlay Mat Instance Dynamic()
    Skeletal Mesh=Get SkelMesh()
    Velocity Transform=foot_l
    Velocity Interp Speed=Get Velocity Interp Speed()
    Collision Transform=calf_l
    Invert Collision Transform Axis=true
    Delta Time X=@n1.DeltaTimeX
    Previous Velocity Transform Location=Get R2 Velocity Transform Location()
    Previous Velocity Transform Velocity=Get R2 Velocity Transform Velocity()
    Region=2
   )
   > n3 = Branch(Condition=Get bIsSkyDiving())
   > Set WPO Pho Wind Intensity(WPO Pho Wind Intensity=2)
   > n4 = Set WPO Pho Turbulence(WPO Pho Turbulence:double=1) -> (Output_Get:double)
   n5 = Get Mat Instance Dynamic()
   n6 = Get Overlay Mat Instance Dynamic()
   SetScalarParameterValue(execute=[@n4.then, @n8.then], self=[@n5, @n6], ParameterName=ClothSim_WindStrength, Value=Get WPO Pho Wind Intensity())
   n7 = Get WPO Pho Turbulence() -> (WPO Pho Turbulence:double)
   > SetScalarParameterValue(self=[@n6, @n5], ParameterName=ClothSim_R1_TurbulenceStrength, Value=@n7)
   > SetScalarParameterValue(self=[@n6, @n5], ParameterName=ClothSim_R2_TurbulenceStrength, Value=@n7)
   Set WPO Pho Wind Intensity(execute=@n3.else, WPO Pho Wind Intensity=0)
   > n8 = Set WPO Pho Turbulence(WPO Pho Turbulence=0.2)
   n9 = Gravity Override State Function DEPRECATED(
    execute=@n2.then_1
    CharacterPartAnimInstance=Self-Reference()
    GameTimeInSeconds=GetGameTimeInSeconds()
    GravityOverrideParams=Get Gravity Override Params()
    RBAN Params=Get RBAN Params()
   )
   > Set Gravity Override(Gravity Override=@n9)
   > Set RBAN Settings Out!(execute, RBAN Settings Out:RBANSetting Struct Structure=@n9.RBAN SettingsOut) -> (
    then
    Output_Get_ComponentLinearAccScale_28_#1:Vector=~
    Output_Get_ComponentLinearVelScale_31_#2:Vector=~
    Output_Get_SimSpaceSettings_34_#3:Sim Space Settings Structure
   )
   n10 = NOT Boolean(A=Get bIsMoving2D())
   n11 = Get bIsCrouching()
   > Set Crouching(Crouching:double="To Float (Boolean)"(InBool=AND Boolean(A=@n11, B=@n10))) -> (Output_Get:double)
   n12 = Get bIsCrouching()
   n13 = Get bIsMoving2D()
   > Set CrouchMOV(CrouchMOV:double="To Float (Boolean)"(InBool=AND Boolean(A=@n12, B=@n13))) -> (Output_Get:double)
   n14 = NOT Boolean(A=Get bIsJumping())
   n15 = Get bIsMoving2D()
   > Set is moving(is moving:double="To Float (Boolean)"(InBool=AND Boolean(A=@n15, B=@n14))) -> (Output_Get:double)
   > Set skydivve(skydivve:double="To Float (Boolean)"(InBool=Get bIsSkydiveFloating())) -> (Output_Get:double)
   n16 = NOT Boolean(A=Get bIsSkyDiving())
   n17 = Get bIsJumping()
   > Set Isjumpo(Isjumpo:double="To Float (Boolean)"(InBool=AND Boolean(A=@n17, B=@n16))) -> (Output_Get:double)
   > Set PinLower Back(PinLower Back:double="To Float (Boolean)"(InBool=Get bIsBaseballSliding())) -> (Output_Get:double)
   n18 = Get bIsCrouching()
   n19 = Get bIsPlayingMeleeAnim()
   > Set pickaxe(pickaxe:double="To Float (Boolean)"(InBool=AND Boolean(A=@n19, B=@n18))) -> (Output_Get:double)
   n20 = Get bIsPlayingEmote()
   n21 = Get bIsParachuteOpened()
   > Set Emote(Emote:double="To Float (Boolean)"(InBool=OR Boolean(A=@n20, B=@n21))) -> (Output_Get:double)
   > Set RailGrind(RailGrind:double="To Float (Boolean)"(InBool=Get bIsOnGrindRail())) -> (Output_Get:double)
   > Set Knees(Knees:double="To Float (Boolean)"(InBool=Get bIsKneeSliding())) -> (Output_Get:double)
   n22 = BreakTransform(InTransform=GetDeltaTransformFromRefPose(self=GetOwnerSkeletalMeshComponent(), BoneName=spine_03, BaseName=pelvis))
   n23 = Branch(execute=@n2.then_2, Condition="float <= float"(A=BreakRotator(InRot=@n22.Rotation), B=-10) -> (ReturnValue=~))
   > Set spine_rot_r(spine_rot_r:double=1) -> (Output_Get:double)
   Set spine_rot_r(execute=@n23.else, spine_rot_r=0)
   n24 = BreakTransform(InTransform=GetDeltaTransformFromRefPose(self=GetOwnerSkeletalMeshComponent(), BoneName=spine_03, BaseName=pelvis))
   n25 = Branch(execute=@n2.then_3, Condition="float >= float"(A=BreakRotator(InRot=@n24.Rotation), B=-10))
   > Set spine_rot_L(spine_rot_L:double=1) -> (Output_Get:double)
   Set spine_rot_L(execute=@n25.else, spine_rot_L=0)
   n26 = GetOwnerSkeletalMeshComponent()
   n27 = BreakRotator(InRot=BreakTransform(InTransform=GetDeltaTransformFromRefPose(self=@n26, BoneName=upperarm_l, BaseName=clavicle_l)).Rotation)
   n28 = Branch(execute=@n2.then_4, Condition="float <= float"(A=@n27.Yaw, B=3) -> (ReturnValue=~))
   > Set UpperArmRot_l(UpperArmRot_l:double=1) -> (Output_Get:double)
   Set UpperArmRot_l(execute=@n28.else, UpperArmRot_l=0)
   n29 = Branch(execute=@n2.then_6, Condition="float <= float"(A=@n27, B=-50) -> (ReturnValue=~))
   > Set spine_rot_D(spine_rot_D:double=1) -> (Output_Get:double)
   Set spine_rot_D(execute=@n29.else, spine_rot_D=0)
   Branch(execute=@n2.then_5, Condition=NOT Boolean(A=Get bIsCrouching()))
   > n30 = PlayerMovement(AnimInstance=Self-Reference(), XDotProductCoefficient=100, YDotProductCoefficient=100, ZDotProductCoefficient=100)
   n31 = BreakVector(InVec=MakeVector(X=@n30, Y=@n30.YMovement))
   > Set Forward(Forward:double="To Float (Boolean)"(InBool="float > float"(A=@n31.Y, B=40))) -> (Output_Get:double)
   > Set notForward(notForward:double="To Float (Boolean)"(InBool="float < float"(A=@n31.Y, B=-40))) -> (Output_Get:double)
   Event OnBeginJumping()
   > Set jumping(jumping:bool=true) -> (Output_Get:bool)
   Event OnCharacterAndPickaxeCustomizationLoaded!() -> (
    OutputDelegate
    then
    CustomBaseMesh:Skeletal Mesh Component Object Reference
    CustomParts:Array of Custom Character Part Object References
    CustomMeshParts:Array of Skeletal Mesh Component Budgeted Object References
   ) {is=title K2Node_Event}
   > Do Once()
   n32 = CheckForBackbling()
   > n33 = Branch(Condition=OR Boolean!(A=@n32, B=@n32.bIsBackblingACape, C:bool=@n32.bIsBackblingLong) -> (ReturnValue))
   > Set isWearingBackbling(isWearingBackbling=true)
   Set isWearingBackbling(execute=@n33.else, isWearingBackbling=false)
   Event OnCharacterCustomizationLoaded()
   > n34 = Set SkelMesh(SkelMesh=GetPartSkeletalMeshComponent())
   > n35 = Get Dynamic Material Instance(SkeletalMeshComponent=@n34)
   > Set Mat Instance Dynamic(Mat Instance Dynamic=@n35)
   > n36 = GetOrCreateDynamicOverlayMaterial(MeshComponent=@n34)
   > Set Overlay Mat Instance Dynamic(Overlay Mat Instance Dynamic=@n36)
   Event OnEndJumping()
   > Set jumping(jumping=false)
   Event BlueprintInitializeAnimation()
   > Set SkelMesh(SkelMesh=GetOwnerSkeletalMeshComponent())
   n37 = TryGetPawnOwner()
   > Is Valid(InputObject=@n37)
   > n38 = Cast To FortPlayerPawnAthena(Object=@n37)
   > n39 = Set PlayerPawn(PlayerPawn:Fort Player Pawn Athena Object Reference=@n38) -> (Output_Get:Fort Player Pawn Athena Object Reference)
   > Set Body_Mesh(Body_Mesh=GetSkeletalMeshForPartType(self=@n39))
   SetScalarParameterValue(self=[@n6, @n5], ParameterName=ClothSim_R3_TurbulenceStrength, Value=@n7)
   "float <= float"(B=-15) -> (ReturnValue=~)
   Get bIsPlayingEmote()
   Get bIsJumping()
   Set Moving(Moving:bool=NOT Boolean(A=Get bIsMoving2D())) -> (Output_Get:bool)
  ]}
]
variables [
 BagLift {type=number}
 Body_Mesh {type=SkeletalMeshComponent}
 Crouching {type=number}
 CrouchMOV {type=number}
 Emote {type=number}
 Forward {type=number}
 Gravity Override {type=Vector}
 Gravity Override Params {type=GravityOverrideParamsStruct}
 is moving {type=number}
 Isjumpo {type=number}
 isWearingBackbling {type=boolean}
 jumping {type=boolean}
 Knees {type=number}
 Local Noise Offset {type=Vector}
 Local Noise Transform Location {type=Vector}
 Mat Instance Dynamic {type=MaterialInstanceDynamic}
 Moving {type=boolean}
 notForward {type=number}
 Overlay Mat Instance Dynamic {type=MaterialInstanceDynamic}
 pickaxe {type=number}
 PinLower Back {type=number}
 PlayerPawn {type=FortPlayerPawnAthena}
 R1 Velocity Transform Location {type=Vector}
 R1 Velocity Transform Velocity {type=Vector}
 R2 Velocity Transform Location {type=Vector}
 R2 Velocity Transform Velocity {type=Vector}
 RailGrind {type=number}
 RBAN Params {
  type=RBANParamsStruct
  default {
   dBNO {simSpaceSettings {
     maxAngularAcceleration=1000
     maxAngularVelocity=3
     maxLinearAcceleration=300
     maxLinearVelocity=700
     velocityScaleZ=0
     externalLinearDragV=<0 0.5 0>
    }}
   dBNO_Carried {simSpaceSettings {
     maxAngularAcceleration=1000
     maxAngularVelocity=3
     maxLinearAcceleration=300
     maxLinearVelocity=700
     velocityScaleZ=0
     externalLinearDragV=<0 0.5 0>
    }}
   emoteOrMelee {simSpaceSettings {
     maxAngularAcceleration=1000
     maxAngularVelocity=3
     maxLinearAcceleration=300
     maxLinearVelocity=700
     velocityScaleZ=0
     externalLinearDragV=<0 0.5 0>
    }}
   front_End {simSpaceSettings {
     maxAngularAcceleration=1000
     maxAngularVelocity=3
     maxLinearAcceleration=300
     maxLinearVelocity=700
     velocityScaleZ=0
     externalLinearDragV=<0 0.5 0>
    }}
   onGround_Crouch_Moving {simSpaceSettings {
     maxAngularAcceleration=1000
     maxAngularVelocity=3
     maxLinearAcceleration=300
     maxLinearVelocity=700
     velocityScaleZ=0
     externalLinearDragV=<0 0.5 0>
    }}
   onGround_Falling {simSpaceSettings {
     maxAngularAcceleration=1000
     maxAngularVelocity=3
     maxLinearAcceleration=300
     maxLinearVelocity=700
     velocityScaleZ=0
     externalLinearDragV=<0 0.5 0>
    }}
   onGround_Moving {simSpaceSettings {
     maxAngularAcceleration=1000
     maxAngularVelocity=3
     maxLinearAcceleration=300
     maxLinearVelocity=700
     velocityScaleZ=0
     externalLinearDragV=<0 0.5 0>
    }}
   onGround_Standing {simSpaceSettings {
     maxAngularAcceleration=1000
     maxAngularVelocity=3
     maxLinearAcceleration=300
     maxLinearVelocity=700
     velocityScaleZ=0
     externalLinearDragV=<0 0.5 0>
    }}
   rail_Grinding {simSpaceSettings {
     maxAngularAcceleration=1000
     maxAngularVelocity=3
     maxLinearAcceleration=300
     maxLinearVelocity=700
     velocityScaleZ=0
     worldAlpha=0.1
     externalLinearDragV=<0 0.5 0>
    }}
   riding_Moving {simSpaceSettings {
     maxAngularAcceleration=1000
     maxAngularVelocity=3
     maxLinearAcceleration=300
     maxLinearVelocity=700
     velocityScaleZ=0
     externalLinearDragV=<0 0.5 0>
    }}
   riding_Moving_Targeting {simSpaceSettings {
     maxAngularAcceleration=1000
     maxAngularVelocity=3
     maxLinearAcceleration=300
     maxLinearVelocity=700
     velocityScaleZ=0
     externalLinearDragV=<0 0.5 0>
    }}
   riding_Standing {simSpaceSettings {
     maxAngularAcceleration=1000
     maxAngularVelocity=3
     maxLinearAcceleration=300
     maxLinearVelocity=700
     velocityScaleZ=0
     externalLinearDragV=<0 0.5 0>
    }}
   skydive_DivingDown {simSpaceSettings {
     maxAngularAcceleration=1000
     maxAngularVelocity=3
     maxLinearAcceleration=300
     maxLinearVelocity=700
     velocityScaleZ=0
     externalLinearDragV=<0 0.5 0>
    }}
   skydive_FloatingDown {
    componentLinearAccScale=<0 0.5 0>
    componentLinearVelScale=<0 0.5 0>
    simSpaceSettings {
     maxAngularAcceleration=1000
     maxAngularVelocity=3
     maxLinearAcceleration=300
     maxLinearVelocity=700
     velocityScaleZ=0
     externalLinearDragV=<0 0.5 0>
    }
   }
   skydive_GoingUp {simSpaceSettings {
     maxAngularAcceleration=1000
     maxAngularVelocity=3
     maxLinearAcceleration=300
     maxLinearVelocity=700
     velocityScaleZ=0
     externalLinearDragV=<0 0.5 0>
    }}
   skydive_Parachuting {
    componentLinearAccScale=<0 0.1 0.3>
    componentLinearVelScale=<0 0.1 0.3>
    simSpaceSettings {
     maxAngularAcceleration=1000
     maxAngularVelocity=3
     maxLinearAcceleration=300
     maxLinearVelocity=700
     velocityScaleZ=0
     externalLinearDragV=<0 0.5 0>
    }
   }
   skydive_Surfing {simSpaceSettings {
     maxAngularAcceleration=1000
     maxAngularVelocity=3
     maxLinearAcceleration=300
     maxLinearVelocity=700
     velocityScaleZ=0
     externalLinearDragV=<0 0.5 0>
    }}
   swimming {simSpaceSettings {
     maxAngularAcceleration=1000
     maxAngularVelocity=3
     maxLinearAcceleration=300
     maxLinearVelocity=700
     velocityScaleZ=0
     externalLinearDragV=<0 0.5 0>
    }}
   tactical_Sprint {simSpaceSettings {
     maxAngularAcceleration=1000
     maxAngularVelocity=3
     maxLinearAcceleration=300
     maxLinearVelocity=700
     velocityScaleZ=0
     externalLinearDragV=<0 0.5 0>
    }}
   ziplining {simSpaceSettings {
     maxAngularAcceleration=1000
     maxAngularVelocity=3
     maxLinearAcceleration=300
     maxLinearVelocity=700
     velocityScaleZ=0
     externalLinearDragV=<0 0.5 0>
    }}
  }
 }
 RBAN Settings Out {
  type=RBANSettingStruct
  default {
   componentLinearAccScale=<0 0 0>
   componentLinearVelScale=<0 0 0>
   simSpaceSettings {
    dampingAlpha=1
    maxAngularAcceleration=1000
    maxAngularVelocity=3
    maxLinearAcceleration=300
    maxLinearVelocity=700
    velocityScaleZ=0
    worldAlpha=1
    externalAngularVelocity=<0 0 0>
    externalLinearDragV=<0 0.5 0>
    externalLinearVelocity=<0 0 0>
   }
  }
 }
 SkelMesh {type=SkeletalMeshComponent}
 skydivve {type=number}
 spine_rot_D {type=number}
 spine_rot_L {type=number}
 spine_rot_r {type=number}
 UpperArmRot_l {type=number}
 Velocity Interp Speed {type=number}
 WPO Pho Turbulence {type=number}
 WPO Pho Wind Intensity {type=number}
]
