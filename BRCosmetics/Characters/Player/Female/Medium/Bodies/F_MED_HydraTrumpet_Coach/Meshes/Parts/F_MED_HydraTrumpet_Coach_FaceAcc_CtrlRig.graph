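# Control Rig for the HydraTrumpet Coach face accessory, applying corrective offsets to hair and scruff bones as the head moves.
# Two spherical pose readers measure head orientation relative to spine_04, storing NeckbendAlpha and IsTargetingAlpha.
# NeckbendAlpha weights a back-hair offset.
# Its product with the separate istargeting variable is interpolated to weight further back-hair and scruff offsets, while IsTargetingAlpha is never read in the graph.
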ControlRigBlueprint /BRCosmetics/Characters/Player/Female/Medium/Bodies/F_MED_HydraTrumpet_Coach/Meshes/Parts/F_MED_HydraTrumpet_Coach_FaceAcc_CtrlRig
graphs [
 RigVMFunctionLibrary {links [], nodes []}
 RigVMModel {nodes [
   RigUnit_BeginExecution = &RigUnit_BeginExecution()
   RigVMFunction_Sequence = &RigVMFunction_Sequence(ExecuteContext=@RigUnit_BeginExecution.ExecutePin)
   n1 = &RigUnit_SphericalPoseReader(
    ActiveRegionScaleFactors=(PositiveWidth=3,NegativeWidth=0.4,PositiveHeight=6,NegativeHeight=6)
    ActiveRegionSize=0.1
    Debug=(bDrawDebug=false,DebugScale=25,DebugSegments=20,DebugThickness=0.25)
    DriverAxis=(X=1,Y=0,Z=0)
    DriverItem=(Type=Bone,Name="head")
    FalloffRegionScaleFactors=(PositiveWidth=-3,NegativeWidth=0.8,PositiveHeight=0,NegativeHeight=0)
    FalloffSize=0.25
    OptionalParentItem=(Type=Bone,Name="spine_04")
    RotationOffset=(X=90,Y=0,Z=-140)
    ExecutePin=@RigVMFunction_Sequence.A
   )
   > RigVMVariableNode(Value=@n1.OutputParam) {variable=NeckbendAlpha}
   > n2 = &RigUnit_SphericalPoseReader(
    ActiveRegionScaleFactors=(PositiveWidth=1,NegativeWidth=0.3,PositiveHeight=1,NegativeHeight=1)
    ActiveRegionSize=0.1
    Debug=(bDrawDebug=false,DebugScale=25,DebugSegments=20,DebugThickness=0.25)
    DriverAxis=(X=1,Y=0,Z=0)
    DriverItem=(Type=Bone,Name="head")
    FalloffRegionScaleFactors=(PositiveWidth=0.5,NegativeWidth=0.2,PositiveHeight=0.5,NegativeHeight=0.5)
    FalloffSize=0.25
    FlipWidthScaling=false
    OptionalParentItem=(Type=Bone,Name="spine_04")
    RotationOffset=(X=40,Y=20,Z=-140)
   )
   > RigVMVariableNode(Value=@n2.OutputParam) {variable=IsTargetingAlpha}
   &RigUnit_ModifyTransforms(
    ItemToModify=((Item=(Type=Bone,Name="base_backhair"),Transform=(Rotation=(X=0,Y=0,Z=-0.2164,W=0.9763),Translation=(X=5,Y=0,Z=0),Scale3D=(X=1,Y=1,Z=1))))
    WeightMinimum=0
    ExecutePin=@RigVMFunction_Sequence.B
    Weight=RigVMVariableNode() {variable=NeckbendAlpha}.Value
   )
   Multiply = &RigVMFunction_MathDoubleMul(B=RigVMVariableNode() {variable=istargeting}.Value, A=RigVMVariableNode() {variable=NeckbendAlpha}.Value)
   n3 = &RigVMFunction_AlphaInterp(InterpSpeedDecreasing=10, InterpSpeedIncreasing=10, Scale=1, bClampResult=true, bMapRange=true, Value=@Multiply.Result)
   > &RigUnit_ModifyTransforms(
    ItemToModify="((Item=(Type=Bone,Name=\"base_backhair\"),Transform=(Rotation=(X=0.0302,Y=-0.171,Z=0.171,W=0.9698),Translation=(X=1,Y=0,Z=0),Scale3D=(X=1,Y=1,Z=1))),(Item=(Type=Bone,Name=\"base_scruff_r\"),Transform=(Rotation=(X=0,Y=0,Z=-0.5,W=0.866),Translation=(X=1,Y=-1,Z=0 ...(truncated)"
    WeightMinimum=0
    Weight=@n3.Result
   )
  ]}
]
hierarchy [
 root {loc=<0 0 0>, rot=<0 0 0 1>}
 pelvis {parent=root, loc=<0 -0.5301 98.0571>, rot=<0.0127 0.707 -0.0127 -0.707>}
 spine_01 {parent=pelvis, loc=<6.1208 -0.9328 0>, rot=<0 0 0.034 -0.9994>}
 spine_02 {parent=spine_01, loc=<6.5997 -0.2095 0>, rot=<0 0 0.0238 -0.9997>}
 spine_03 {parent=spine_02, loc=<6.4914 -0.1237 0>, rot=<0 0 0.0149 -0.9999>}
 spine_04 {parent=spine_03, loc=<6.9598 0.0794 0>, rot=<0 0 -0.0132 -0.9999>}
 spine_05 {parent=spine_04, loc=<6.9987 0 0>, rot=<0 0 -0.0015 -1>}
 neck_01 {parent=spine_05, loc=<11.3138 0.1529 0>, rot=<0 0 -0.2003 -0.9797>}
 neck_02 {parent=neck_01, loc=<4.5627 0 0>, rot=<0 0 0.0504 -0.9987>}
 head {parent=neck_02, loc=<5.3914 0 0>, rot=<0 0 0.115 -0.9934>}
 base_strands {parent=head, loc=<18.9871 -1.2828 -0.1204>, rot=<-0.7196 0 0 -0.6944>}
 dyn_strands_b {parent=base_strands, loc=<-3.2288 1.0982 8.0919>, rot=<0 0 0 1>}
 dyn_strands_l {parent=base_strands, loc=<-0.6197 -7.4888 -4.1283>, rot=<0 0 0 1>}
 dyn_strands_r {parent=base_strands, loc=<-0.5595 5.9488 -6.9837>, rot=<0 0 0 1>}
 base_backhair {parent=head, loc=<-1.9553 -9.0472 -0.4841>, rot=<0.9956 0 0 -0.0942>}
 dyn_backhair_r {parent=base_backhair, loc=<-0.1423 0.4389 3.433>, rot=<0 0 0 1>}
 dyn_backhair_l {parent=base_backhair, loc=<0.0379 -1.2992 -3.2136>, rot=<0 0 0 1>}
 dyn_backhair_f {parent=base_backhair, loc=<-2.8112 1.239 -0.6282>, rot=<0 0 0 1>}
 base_scruff_l {parent=head, loc=<-3.5681 -1.4692 -8.8959>, rot=<0.7378 0 0 -0.675>}
 dyn_scruff_l_r {parent=base_scruff_l, loc=<-0.1423 0.4389 3.433>, rot=<0 0 0 1>}
 dyn_scruff_l_l {parent=base_scruff_l, loc=<0.4893 -0.3608 -3.4208>, rot=<0 0 0 1>}
 dyn_scruff_L_f {parent=base_scruff_l, loc=<-0.3297 1.848 -0.7445>, rot=<0 0 0 1>}
 base_scruff_r {parent=head, loc=<-3.5681 -1.4692 9.0087>, rot=<-0.7196 0 0 -0.6944>}
 dyn_scruff_r_r {parent=base_scruff_r, loc=<-0.5439 -0.8432 3.4878>, rot=<0 0 0 1>}
 dyn_scruff_r_l {parent=base_scruff_r, loc=<1.1145 1.4253 -5.59>, rot=<0 0 0 1>}
 dyn_scruff_r_f {parent=base_scruff_r, loc=<-0.3297 1.848 -0.7445>, rot=<0 0 0 1>}
 base_beard {parent=head, loc=<-6.4353 7.9478 0>, rot=<0 0 0 1>}
 dyn_beard_r {parent=base_beard, loc=<-1.1433 -0.0095 2.9245>, rot=<0 0 0 1>}
 dyn_deard_l {parent=base_beard, loc=<-0.7761 -0.666 -3.9159>, rot=<0 0 0 1>}
 dyn_beard_f {parent=base_beard, loc=<-1.7827 2.4655 -0.7545>, rot=<0 0 0 1>}
 dyn_beard_b {parent=base_beard, loc=<-2.375 -1.0764 0>, rot=<0 0 0 1>}
]
variables [
 IsTargetingAlpha {type=double, default=0}
 NeckbendAlpha {type=double, default=0}
 istargeting {type=double, default=0}
]
